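# Editor utility sub-widget for the companion anim asset generator that shows one subfolder column of animation names under a header checkbox.
# SetupList matches the creature's archetype to a DT_CompanionArchetype row, then finds the SubFolders entry matching the column name.
# It creates an AnimListItem in RowHolder for each animation name.
# The header checkbox and CheckPassThrough event call TriggerCheckBox, which sets every item's IncludeCheck to match.
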EditorUtilityWidgetBlueprint /AnimInEdTools/CompanionAnimAssetGenerator/SubWidgets/AnimLists
graphs [
 EventGraph {nodes [
   n1 = "On Check State Changed (AllColumnItemsCheckBox)"!() -> (OutputDelegate, then, bIsChecked:bool) {is=title K2Node_ComponentBoundEvent}
   > TriggerCheckBox!(execute, self, InputPin:bool=@n1.bIsChecked) -> (then)
   n2 = CheckPassThrough!() -> (OutputDelegate, then, Checked:bool=~) {is=title K2Node_CustomEvent}
   > TriggerCheckBox(InputPin=@n2.Checked)
   Event PreConstruct()
   Event Construct()
   Event Tick()
  ]}
 SetupList {nodes [
   n1 = SetupList!() -> (then, ColumnName:String, AC Ref:Animatable Creature Object Reference, SubFolderLocation:int) {is=title K2Node_FunctionEntry}
   > Set Text(Text="To Text (String)"(InString=@n1), self=Get ColumnName() -> (ColumnName:Text Object Reference))
   > n2 = GetDataTableRowNames(Table=/AnimInEdTools/CompanionAnimAssetGenerator/Settings/DT_CompanionArchetype)
   n3 = Get AC Ref() -> (AC Ref:Animatable Creature Object Reference)
   n4 = String To Name(InString=Enum to String!(Enumerator:ENUM_CompanionArchetype Enum=Get Archetype!(self:Animatable Creature Object Reference=@n3) -> (Archetype:ENUM_CompanionArchetype Enum=NewEnumerator0)) -> (ReturnValue:String))
   > n5 = Get Data Table Row DT_CompanionArchetype!(
    execute:Exec
    DataTable:Data Table Object Reference=&DT_CompanionArchetype
    RowName:Name="Get (a ref)"(Array:Array of Names=@n2, Dimension 1=Find Item!(TargetArray:Array of Names=@n2, ItemToFind:"Name (by ref)"=@n4) -> (ReturnValue)) -> (Output:"Name (by ref)")
   ) -> (
    then:Exec
    RowNotFound:Exec
    ReturnValue_SubFolders_45_#1:Map of Strings to Struct Companion Archetype Anim Names Structures
    ReturnValue_ArchetypeLeaderAssetPath_51_#2:String
   ) {is=title K2Node_GetDataTableRow}
   > n6 = Keys!(execute, TargetMap:Map of Strings to Struct Companion Archetype Anim Names Structures=@n5) -> (then, Keys:Array of Strings)
   > n7 = For Each Loop(Array=@n6)
   > Branch(Condition="Equal Exactly (String)"(A=@n7, B=Get ColumnName() -> (ColumnName:String)))
   > n8 = Values!(execute, TargetMap:Map of Strings to Struct Companion Archetype Anim Names Structures=@n5) -> (then, Values:Array of Struct Companion Archetype Anim Names Structures)
   > n9 = For Each Loop(Array="Get (a ref)"!(Array:Array of Struct Companion Archetype Anim Names Structures=@n8, Dimension 1=@n7.Array Index) -> (Output_AnimNames_3_#3:Array of Strings))
   > n10 = Create Widget!(
    execute:Exec
    WorldContextObject:Object Reference
    Class:User Widget Class Reference=&/AnimListItem.AnimListItem_C
    OwningPlayer:Player Controller Object Reference
   ) -> (then:Exec, ReturnValue:Anim List Item Object Reference)
   > AddChildToStackBox!(execute, self:Stack Box Object Reference=Get RowHolder() -> (RowHolder:Stack Box Object Reference), Content:Widget Object Reference=@n10) -> (then, ReturnValue:Stack Box Slot Object Reference)
   > ListItemSetup!(
    execute
    self:Anim List Item Object Reference=@n10
    RowName:String=@n9
    AC Ref:Animatable Creature Object Reference=Get AC Ref()
    ColumnName:String=Get ColumnName()
   ) -> (then)
  ]}
 TriggerCheckBox {nodes [
   n1 = TriggerCheckBox!() -> (then, InputPin:bool=~) {is=title K2Node_FunctionEntry}
   > n2 = For Each Loop(Array:Array of Widget Object References=GetAllChildren(self=Get RowHolder())) -> (Array Element:Widget Object Reference)
   > n3 = Cast To AnimListItem!(execute, Object=@n2) -> (then, CastFailed, AsAnim List Item:Anim List Item Object Reference) {is=title K2Node_DynamicCast}
   > n4 = Branch(Condition=@n1)
   n5 = Get IncludeCheck!(self:Anim List Item Object Reference=@n3) -> (IncludeCheck:Editor Utility Check Box Object Reference)
   > SetCheckedState!(execute, self:Check Box Object Reference=@n5, InCheckedState:ECheckBoxState Enum=Checked) -> (then)
   SetCheckedState(execute=@n4.else, self=@n5, InCheckedState=Unchecked)
  ]}
]
variables [CompanionAnimDataTable {type=DataTable, default=&DT_CompanionArchetype}]
widgets {
 info {rootWidgetClass=&VerticalBox, widgetCount=9}
 widgets [
  VerticalBox_0 {widgetClassPath=&VerticalBox, widget=VerticalBox_0}
  HeaderBorder {
   widgetClassPath=&Border
   parent=VerticalBox_0
   widget=HeaderBorder
   properties {brushColor=<rgba 0.0275 0.0275 0.0275 1>, padding=<ltrb 2 2 2 2>}
   slot {class=VerticalBoxSlot, properties {padding=<ltrb 0 0 0 1>}}
  }
  VerticalBox_1 {
   widgetClassPath=&VerticalBox
   parent=HeaderBorder
   widget=VerticalBox_1
   slot {class=BorderSlot, properties {padding=<ltrb 2 2 2 2>}}
  }
  HorizontalBox_1 {widgetClassPath=&HorizontalBox, parent=VerticalBox_1, widget=HorizontalBox_1, slot {class=VerticalBoxSlot}}
  AllColumnItemsCheckBox {
   widgetClassPath=&EditorUtilityCheckBox
   bIsVariable=true
   parent=HorizontalBox_1
   widget=AllColumnItemsCheckBox
   slot {class=HorizontalBoxSlot, properties {horizontalAlignment=HAlign_Center, verticalAlignment=VAlign_Center, size {sizeRule=Fill, value=1}}}
  }
  ColumnName {
   widgetClassPath=&TextBlock
   bIsVariable=true
   parent=AllColumnItemsCheckBox
   widget=ColumnName
   properties {justification=Center, text=SubFolderNameHere}
   slot {class=PanelSlot}
  }
  VerticalBox_48 {
   widgetClassPath=&VerticalBox
   parent=VerticalBox_1
   widget=VerticalBox_48
   slot {class=VerticalBoxSlot, properties {size {sizeRule=Fill, value=1}}}
  }
  HorizontalBox_386 {
   widgetClassPath=&HorizontalBox
   parent=VerticalBox_48
   widget=HorizontalBox_386
   slot {class=VerticalBoxSlot, properties {size {sizeRule=Fill, value=1}}}
  }
  RowHolder {
   widgetClassPath=&StackBox
   bIsVariable=true
   parent=HorizontalBox_386
   widget=RowHolder
   properties {orientation=Orient_Vertical}
   slot {class=HorizontalBoxSlot, properties {size {sizeRule=Fill, value=1}}}
  }
 ]
}
