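# Editor Utility Widget that generates animation assets for companion creatures: Level Sequences, linked Anim Sequences, montages, Blend Spaces and Data Assets.
# Creatures are spawned with Control Rig tracks, and Python bakes linked sequences, using marked frames from the archetype leader to split AnimSETs.
# Data tables drive additive detection and Python-built Blend Spaces from the generated animations.
# Data Assets are duplicated from the archetype leader, then filled through per-archetype structs.
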EditorUtilityWidgetBlueprint /AnimInEdTools/CompanionAnimAssetGenerator/CompanionAnimAssetGenerator
graphs [
 Add Actor {nodes [
   n1 = Add Actor!() -> (then, Animatable Creature:Animatable Creature Object Reference, Level Sequence:Level Sequence Object Reference, bIsLayeredControlRig:bool=~) {is=title K2Node_FunctionEntry}
   > n2 = LoadClassAsset_Blocking(AssetClass=Get Sequencer Actor!(self:Animatable Creature Object Reference=@n1) -> (Sequencer Actor:Object Soft Class Reference))
   > n3 = Load Asset Blocking!(
    execute
    InRef:"Control Rig Asset Soft Reference Structure (by ref)"=Get Control Rig!(self:Animatable Creature Object Reference=@n1) -> (Control Rig:Control Rig Asset Soft Reference Structure)
    bForceLoad:bool=true
   ) -> (then, ReturnValue:Control Rig Asset Strong Reference Structure)
   > n4 = Add Actor Base!(
    execute
    self
    Level Sequence:Level Sequence Object Reference=Get Level Sequence() -> (Level Sequence:Level Sequence Object Reference)
    bIsLayeredControlRig:bool=Get bIsLayeredControlRig() -> (bIsLayeredControlRig:bool)
    ControlRigClass:Control Rig Asset Strong Reference Structure=@n3
    Actor Class:Object Class Reference=@n2
   ) -> (then, New Binding:Movie Scene Binding Proxy Structure)
   > Return Node!(execute, New Binding:Movie Scene Binding Proxy Structure=@n4)
  ]}
 Add Actor Base {nodes [
   n1 = Add Actor Base!() -> (
    then
    Level Sequence:Level Sequence Object Reference
    bIsLayeredControlRig:bool=~
    ControlRigClass:Control Rig Asset Strong Reference Structure
    Actor Class:Object Class Reference
   ) {is=title K2Node_FunctionEntry}
   > n2 = AddSpawnableFromClass!(
    execute
    self:Level Sequence Editor Subsystem Object Reference=Get LevelSequenceEditorSubsystem() -> (ReturnValue:Level Sequence Editor Subsystem Object Reference) {is=title K2Node_GetEditorSubsystem}
    Sequence:Movie Scene Sequence Object Reference=@n1
    ClassToSpawn:Object Class Reference=@n1.Actor Class
   ) -> (then, ReturnValue:Movie Scene Binding Proxy Structure)
   > n3 = Branch(Condition=Is Valid Referfence!(InRef:"Control Rig Asset Strong Reference Structure (by ref)"=Get ControlRigClass() -> (ControlRigClass:Control Rig Asset Strong Reference Structure)) -> (ReturnValue:bool))
   > n4 = GetEditorWorld(self=Get UnrealEditorSubsystem())
   > FindOrCreateControlRigTrackFromReference!(
    execute
    World:World Object Reference=@n4
    LevelSequence:Level Sequence Object Reference=Get Level Sequence()
    InRef:"Control Rig Asset Strong Reference Structure (by ref)"=Get ControlRigClass()
    InBinding:"Movie Scene Binding Proxy Structure (by ref)"=@n2
    bIsLayeredControlRig:bool=Get bIsLayeredControlRig()
   ) -> (then, ReturnValue:Movie Scene Track Object Reference)
   > Return Node(New Binding=@n2)
   Return Node(execute=@n3.else, New Binding=@n2)
  ]}
 CheckForAnimSET {nodes [
   CheckForAnimSET() {is=title K2Node_FunctionEntry}
   n1 = Get CurrentLvlSeq() -> (CurrentLvlSeq:Level Sequence Object Reference)
   > Branch(Condition=Contains(SearchIn=GetDisplayName(Object=@n1), Substring=AnimSET))
   > CreateAnimSeqAnimSET()
  ]}
 CheckNeedsAdditive {nodes [
   n1 = CheckNeedsAdditive!() -> (then, AnimSequence:Anim Sequence Object Reference) {is=title K2Node_FunctionEntry}
   > n2 = Branch(Condition=Contains(SearchIn=GetDisplayName(Object=@n1), Substring=animSET))
   n3 = GetDataTableRowNames(execute=@n2.else, Table=/AnimInEdTools/CompanionAnimAssetGenerator/Settings/DT_AdditiveAnimSettings)
   > n4 = For Each Loop with Break!(Exec:Exec, Array:Array of Names=@n3, Break:Exec=@n7.then) -> (LoopBody:Exec, Array Element:Name, Array Index:int, Completed:Exec)
   Return Node!(
    execute=@n4.Completed
    HasAdditive:bool=Get found() -> (found:bool)
    AnimSeq:Anim Sequence Object Reference=Get AnimSequence() -> (AnimSequence:Anim Sequence Object Reference)
    AdditiveSettings:Struct Additive Anim Settings Structure=Get Additives() -> (Additives:Struct Additive Anim Settings Structure)
   )
   n5 = Get AnimSequence()
   > Branch(Condition=Contains(SearchIn=GetDisplayName(Object=@n5), Substring="To String (Name)"(InName=@n4), bUseCase=true, bSearchFromEnd=true))
   > n6 = Get Data Table Row DT_AdditiveAnimSettings!(
    execute:Exec
    DataTable:Data Table Object Reference=&DT_AdditiveAnimSettings
    RowName:Name=@n4
   ) -> (then:Exec, RowNotFound:Exec, ReturnValue:Struct Additive Anim Settings Structure) {is=title K2Node_GetDataTableRow}
   > Set Additives(Additives:Struct Additive Anim Settings Structure=@n6) -> (Output_Get:Struct Additive Anim Settings Structure)
   > n7 = Set found(found:bool=true) -> (Output_Get:bool)
  ]}
 CheckNeedsMontage {nodes [
   CheckNeedsMontage() {is=title K2Node_FunctionEntry}
   > Branch(Condition=IsChecked(self=Get MontageCheckBox() -> (MontageCheckBox:Editor Utility Check Box Object Reference)))
   > n1 = "For Each Loop (Map)"!(
    execute:Exec
    MapPin:Map of Strings to Anim Sequence Object References=Get AnimCollection() -> (AnimCollection:Map of Strings to Anim Sequence Object References)
    BreakPin:Exec
   ) -> (then:Exec, KeyPin:String, ValuePin:Anim Sequence Object Reference, CompletedPin:Exec)
   n2 = GetDisplayName(Object=@n1.ValuePin)
   > Branch(Condition=OR Boolean!(
     A=Contains(SearchIn=@n2, Substring=interact)
     B=Contains(SearchIn=@n2, Substring=RCT_Angry)
     C:bool=Contains(SearchIn=@n2, Substring=RCT_Dance_In)
     D:bool=Contains(SearchIn=@n2, Substring=RCT_Worried_In)
     E:bool=Contains(SearchIn=@n2, Substring=RCT_NonDance_In)
     F:bool=Contains(SearchIn=@n2, Substring=RCT_Happy_01_In)
     G:bool=Contains(SearchIn=@n2, Substring=RCT_Happy_02_In)
    ) -> (ReturnValue))
   n3 = Get Object Path String(Object=@n1.ValuePin)
   n4 = Append(A=@n2, B=_M)
   > n5 = LoadAsset(AssetPath=Append!(A=GetPath!(InPath:String=@n3) -> (ReturnValue:String), B=/, C:String=@n4, D:String=., E:String=@n4) -> (ReturnValue))
   > n6 = Is Valid(InputObject=@n5)
   CreateMontage!(
    execute=@n6.Is Not Valid
    self
    AnimSequence:Anim Sequence Object Reference=@n1.ValuePin
    ObjectPath:String=Get Object Path String(Object=@n1.ValuePin)
   ) -> (then)
  ]}
 Create Linked Anim Sequence {nodes [
   Create Linked Anim Sequence!() -> (
    then
    Level Sequence:Level Sequence Object Reference
    New AnimSequence Path:String
    Binding:Movie Scene Binding Proxy Structure
    CustomFrameRange:bool=~
    RangeStart:Frame Number Structure
    RangeEnd:Frame Number Structure
    FrameRate:Frame Rate Structure
   ) {is=title K2Node_FunctionEntry}
   > n1 = Get Current Time!(execute, TimeUnit:EMovieSceneTimeUnit Enum=DisplayRate) -> (then, ReturnValue:Movie Scene Sequence Playback Params Structure) {is=title K2Node_CallFunction}
   n2 = Break Movie Scene Sequence Playback Params!(MovieSceneSequencePlaybackParams:"Movie Scene Sequence Playback Params Structure (by ref)"=@n1) -> (
    Frame_FrameNumber_Value:int
    Frame_SubFrame:float=~
    Time:float=~
    MarkedFrame:String
    Timecode:Timecode Structure
    PositionType:EMovieScenePositionType Enum
    UpdateMethod:EUpdatePositionMethod Enum
    bHasJumped:bool=~
   ) {is=title K2Node_BreakStruct}
   > Set Current Time!(
    execute
    PlaybackParams:Movie Scene Sequence Playback Params Structure=Make Movie Scene Sequence Playback Params!(
     Frame_FrameNumber_Value:int=int + int(A=@n2)
     Frame_SubFrame:float=0
     Time:float=0
     MarkedFrame:String
     Timecode:Timecode Structure=(Hours=0,Minutes=0,Seconds=0,Frames=0,Subframe=0,bDropFrameFormat=False)
     PositionType:EMovieScenePositionType Enum=Frame
     UpdateMethod:EUpdatePositionMethod Enum=Play
     bHasJumped:bool=False
    ) -> (MovieSceneSequencePlaybackParams:Movie Scene Sequence Playback Params Structure) {is=title K2Node_MakeStruct}
    TimeUnit:EMovieSceneTimeUnit Enum=DisplayRate
   ) -> (then) {is=title K2Node_CallFunction}
   > Set Current Time(PlaybackParams=@n1, TimeUnit=DisplayRate)
   > n3 = GetBindingID!(
    execute
    Sequence:Movie Scene Sequence Object Reference=Get Level Sequence()
    InBinding:"Movie Scene Binding Proxy Structure (by ref)"=Get Binding() -> (Binding:Movie Scene Binding Proxy Structure)
   ) -> (then, ReturnValue:Movie Scene Object Binding ID Structure)
   > n4 = For Each Loop with Break(
    Array:Array of Object References=GetBoundObjects!(ObjectBinding:Movie Scene Object Binding ID Structure=@n3) -> (ReturnValue:Array of Object References)
    Break=@n21.then
   ) -> (Array Element:Object Reference)
   n5 = Is Valid(exec=@n4.Completed, InputObject=Get Skeleton() -> (Skeleton:Skeleton Object Reference))
   n6 = Get ErrorTxt() -> (ErrorTxt:Text Object Reference)
   SetVisibility(execute=@n5.Is Not Valid, self=@n6, InVisibility=Visible)
   > Set Text(Text="LOCTEXT(\"Skeletal mesh missing from Companion BP\")", self=@n6)
   n7 = Get New AnimSequence Path() -> (New AnimSequence Path:String)
   n8 = Execute Python Script!(
    execute=@n5.Is Valid
    PythonScript="import unreal\r\n\r\n## Grab the Level Editor World\r\neditor_subsystem = unreal.get_editor_subsystem(unreal.UnrealEditorSubsystem)\r\nworld = editor_subsystem.get_editor_world()\r\n \r\n## Create animation sequence export options\r\nanim_seq_export_options = unreal.AnimSeqExportOption()\r\nanim_seq_export_options.export_transforms = True\r\nanim_seq_export_options.export_morph_targets = True\r\nanim_seq_export_options.override_timecode_rate = frame_rate\r\nanim_seq_export_options.use_custom_time_range = custom_range\r\nanim_seq_export_options.custom_start_frame = range_start\r\nanim_seq_export_options.custom_end_frame = range_end\r\nanim_seq_export_options.bake_timecode = True\r\n\r\n \r\n## Get asset tools\r\nasset_tools = unreal.AssetToolsHelpers.get_asset_tools()\r\n\r\n## Create an anim sequence factory, assign it the target skeleton\r\nanim_sequence_factory = unreal.AnimSequenceFactory()\r\nanim_sequence_factory.target_skeleton = skeleton\r\nanim_sequence = unreal.AssetTools.create_asset(asset_tools, asset_name=asset_name, package_path=package_path, asset_class=unreal.AnimSequence, factory=anim_sequence_factory)  \r\n\r\n## Export\r\nunreal.ControlRigSequencerLibrary.export_anim_sequence_from_sequencer(anim_sequence, anim_seq_export_options, binding, True)"
    package_path:String=GetPath(InPath=@n7)
    asset_name:String=GetBaseFilename(InPath=@n7, bRemovePath=true)
    level_sequence:Level Sequence Object Reference=Get Level Sequence()
    binding:Movie Scene Binding Proxy Structure=Get Binding()
    skeleton:Skeleton Object Reference=Get Skeleton()
    CustomRange:bool=Get CustomFrameRange() -> (CustomFrameRange:bool)
    RangeStart:Frame Number Structure=Get RangeStart() -> (RangeStart:Frame Number Structure)
    RangeEnd:Frame Number Structure=Get RangeEnd() -> (RangeEnd:Frame Number Structure)
    frame_rate:Frame Rate Structure=Get FrameRate() -> (FrameRate:Frame Rate Structure)
   ) -> (then, ReturnValue, anim_seq_export_options:Anim Seq Export Option Object Reference) {
    inputs [package_path, asset_name, level_sequence, binding, skeleton, CustomRange, RangeStart, RangeEnd, frame_rate]
    outputs [anim_seq_export_options]
   }
   > Set export options(export options:Anim Seq Export Option Object Reference=@n8.anim_seq_export_options) -> (Output_Get:Anim Seq Export Option Object Reference)
   > n9 = LoadAsset(AssetPath=Get New AnimSequence Path())
   > n10 = Cast To AnimSequence!(execute, Object=@n9) -> (then, CastFailed, AsAnim Sequence:Anim Sequence Object Reference)
   > n11 = Set GeneratedSeq(GeneratedSeq:Anim Sequence Object Reference=@n10) -> (Output_Get:Anim Sequence Object Reference)
   > LinkAnimSequence!(
    execute
    Sequence:Level Sequence Object Reference=Get Level Sequence()
    AnimSequence:Anim Sequence Object Reference=@n11
    ExportOptions:Anim Seq Export Option Object Reference=@n8.anim_seq_export_options
    Binding:"Movie Scene Binding Proxy Structure (by ref)"=Get Binding()
   ) -> (then, ReturnValue:bool)
   > n12 = Sequence()
   n13 = Get Compression Settings!(self:BP Animation Browser Project Object Reference=Get CompanionsProjectRef() -> (CompanionsProjectRef:BP Animation Browser Project Object Reference)) -> (Compression Settings:Anim Sequence Soft Object Reference)
   > Branch(Condition=IsValidSoftObjectReference(SoftObjectReference=@n13))
   > n14 = LoadAsset_Blocking(Asset=@n13)
   > n15 = Cast To AnimSequence(Object=@n14)
   n16 = Get GeneratedSeq() -> (GeneratedSeq:Anim Sequence Object Reference)
   > SetBoneCompressionSettings!(
    execute
    AnimationSequence:Anim Sequence Object Reference=@n16
    CompressionSettings:Anim Bone Compression Settings Object Reference=GetBoneCompressionSettings!(AnimationSequence:Anim Sequence Object Reference=@n15) -> (CompressionSettings:Anim Bone Compression Settings Object Reference)
   ) -> (then)
   > SetCurveCompressionSettings!(
    execute
    AnimationSequence:Anim Sequence Object Reference=@n16
    CompressionSettings:Anim Curve Compression Settings Object Reference=GetCurveCompressionSettings!(AnimationSequence:Anim Sequence Object Reference=@n15) -> (CompressionSettings:Anim Curve Compression Settings Object Reference)
   ) -> (then)
   > SetAnimationInterpolationType!(
    execute
    AnimationSequence:Anim Sequence Object Reference=@n16
    InterpolationType:EAnimInterpolationType Enum=GetAnimationInterpolationType!(AnimationSequence:Anim Sequence Object Reference=@n15) -> (InterpolationType:EAnimInterpolationType Enum=Linear)
   ) -> (then)
   n17 = Get GeneratedSeq()
   AddAnimationModifierOfClass!(execute=@n12.then_1, AnimationSequenceBase:Anim Sequence Base Object Reference=@n17, InModifierClass:Animation Modifier Class Reference) -> (then, ReturnValue:bool)
   > n18 = Is Valid(InputObject=@n17)
   PrintString(execute=@n18.Is Not Valid, InString=FAILED TO GENERATE ANIM SEQ)
   Return Node!(execute=@n18.Is Valid, Anim Sequence:Anim Sequence Object Reference=@n17)
   n19 = Cast To Actor!(execute=@n4.LoopBody, Object=@n4) -> (then, CastFailed, AsActor:Actor Object Reference) {is=title K2Node_DynamicCast}
   n20 = GetComponentByClass!(self:Actor Object Reference=@n19, ComponentClass:Actor Component Class Reference=/Script/Engine.SkeletalMeshComponent) -> (ReturnValue:Skeletal Mesh Component Object Reference)
   > n21 = Set Skeleton(Skeleton:Skeleton Object Reference=Get Skeleton!(self:Skeletal Mesh Object Reference=Get SkeletalMesh!(self:Skinned Mesh Component Object Reference=@n20) -> (SkeletalMesh:Skeletal Mesh Object Reference)) -> (Skeleton)) -> (Output_Get:Skeleton Object Reference)
  ]}
 CreateAnimSeqAnimSET {nodes [
   CreateAnimSeqAnimSET() {is=title K2Node_FunctionEntry}
   n1 = GetBaseFilename(InPath=Get Object Path String(Object=Get CurrentLvlSeq()), bRemovePath=true)
   n2 = Replace(SourceString=GetArchetypeLeaderAssetPath!() -> (LeaderPath:String), From=CosmeticCompanions, To=CompanionsAnimInEd)
   n3 = Get CurrentSubfolder() -> (CurrentSubfolder:String)
   n4 = Get SelectedAC() -> (SelectedAC:Animatable Creature Object Reference)
   > n5 = LoadAsset(AssetPath=Replace(
     SourceString=Append!(A=@n2, B=/Animation/, C=@n3, D=/, E=@n1, F:String=., G:String=@n1) -> (ReturnValue)
     From=Get Asset!(self:Animatable Creature Object Reference=@n4) -> (Asset:String)
     To=GetArchetypeLeaderName!() -> (ReturnValue:String)
    ))
   > n6 = Cast To LevelSequence!(execute, Object=@n5) -> (then, CastFailed, AsLevel Sequence:Level Sequence Object Reference) {is=title K2Node_DynamicCast}
   > n7 = Set ArchetypeLeaderAnimSET(ArchetypeLeaderAnimSET:Level Sequence Object Reference=@n6) -> (Output_Get:Level Sequence Object Reference)
   > n8 = GetPlaybackRange!(execute, Sequence:Movie Scene Sequence Object Reference=@n7) -> (then, ReturnValue:Sequencer Scripting Range Structure)
   > n9 = Set RefSeqRange(RefSeqRange:Sequencer Scripting Range Structure=@n8) -> (Output_Get:Sequencer Scripting Range Structure)
   n10 = Break Sequencer Scripting Range!(SequencerScriptingRange:"Sequencer Scripting Range Structure (by ref)"=@n9) -> (bHasStart:bool=~, bHasEnd:bool=~, InclusiveStart:int, ExclusiveEnd:int) {is=title K2Node_BreakStruct}
   n11 = Get CurrentLvlSeq()
   > SetPlaybackStart!(execute, Sequence:Movie Scene Sequence Object Reference=@n11, StartFrame:int=@n10.InclusiveStart) -> (then)
   > SetPlaybackEnd!(execute, Sequence:Movie Scene Sequence Object Reference=@n11, EndFrame:int=@n10.ExclusiveEnd) -> (then)
   > n12 = Get Marked Frames!(
    execute
    Sequence:Movie Scene Sequence Object Reference=Get ArchetypeLeaderAnimSET() -> (ArchetypeLeaderAnimSET:Level Sequence Object Reference)
    TimeUnit:EMovieSceneTimeUnit Enum=DisplayRate
   ) -> (then, ReturnValue:Array of Movie Scene Marked Frame Structures) {is=title K2Node_CallFunction}
   > n13 = For Each Loop!(Exec, Array:Array of Movie Scene Marked Frame Structures=@n12) -> (
    LoopBody
    Array Element_FrameNumber_Value:int
    Array Element_Label:String
    Array Element_bIsDeterminismFence:bool=~
    Array Element_bIsInclusiveTime:bool=~
    Array Index
    Completed
   )
   n14 = For Each Loop!(Exec=[@n13.Completed, @n24.Completed], Array:Array of Strings=Get MarkerLabels() -> (MarkerLabels:Array of Strings)) -> (LoopBody, Array Element:String, Array Index, Completed)
   Add(
    execute=@n14.Completed
    TargetArray:Array of Object References=Get Assets to save() -> (Assets to save:Array of Object References)
    NewItem:"Object Reference (by ref)"=Get CurrentLvlSeq()
   ) {is=title K2Node_CallArrayFunction}
   > SaveLoadedAssets(AssetsToSave=Get Assets to save(), bOnlyIfIsDirty=true)
   n15 = Create Linked Anim Sequence!(
    execute=@n14.LoopBody
    self
    Level Sequence:Level Sequence Object Reference=Get CurrentLvlSeq()
    New AnimSequence Path:String=Replace(SourceString=Get AnimSeqPath() -> (AnimSeqPath:String), From=AnimSET, To=Replace(SourceString=@n14, From="##"))
    Binding:Movie Scene Binding Proxy Structure=Get CurrentBinding() -> (CurrentBinding:Movie Scene Binding Proxy Structure)
    CustomFrameRange:bool=true
    RangeStart:Frame Number Structure="Get (a ref)"(Array:Array of Frame Number Structures=Get StartFrames() -> (StartFrames:Array of Frame Number Structures), Dimension 1=@n14.Array Index) -> (Output:"Frame Number Structure (by ref)")
    RangeEnd:Frame Number Structure="Get (a ref)"(Array=Get EndFrames() -> (EndFrames:Array of Frame Number Structures), Dimension 1=@n14.Array Index)
    FrameRate:Frame Rate Structure=Get FrameRate()
   ) -> (then, Anim Sequence:Anim Sequence Object Reference) {is=title K2Node_CallFunction}
   > n16 = Is Valid(InputObject=Get CurrentSkeleton() -> (CurrentSkeleton:Skeleton Object Reference))
   n17 = Set CurrentSkeleton(
    execute=@n16.Is Not Valid
    CurrentSkeleton:Skeleton Object Reference=GetSkeleton!(InAsset:Animation Asset Object Reference=@n15) -> (ReturnValue:Skeleton Object Reference)
   ) -> (Output_Get:Skeleton Object Reference)
   Add(execute=[@n17.then, @n16.Is Valid], TargetArray=Get Assets to save(), NewItem=@n15)
   n18 = Branch(execute=@n13.LoopBody, Condition="integer > integer"(A=Len(S=@n13.Array Element_Label), B=1))
   n19 = int * int(A=@n13, B=2)
   > Add!(execute, TargetArray:Array of Frame Number Structures=Get StartFrames(), NewItem_Value:int=@n19) -> (then, ReturnValue)
   > Add!(execute, TargetArray:Array of Strings=Get MarkerLabels(), NewItem:"String (by ref)"=@n13.Array Element_Label) -> (then, ReturnValue)
   > n20 = Branch(Condition="Equal (Integer)"(A=@n13.Array Index))
   Add Marked Frame!(
    execute=[@n21.then, @n20.then]
    Sequence:Movie Scene Sequence Object Reference=Get CurrentLvlSeq()
    InMarkedFrame_FrameNumber_Value:int=@n13
    InMarkedFrame_Label:String=@n13.Array Element_Label
    InMarkedFrame_bIsDeterminismFence:bool=@n13.Array Element_bIsDeterminismFence
    InMarkedFrame_bIsInclusiveTime:bool=@n13.Array Element_bIsInclusiveTime
    TimeUnit:EMovieSceneTimeUnit Enum=DisplayRate
   ) -> (then, ReturnValue:int=0)
   n21 = Add!(execute=[@n18.else, @n20.else], TargetArray:Array of Frame Number Structures=Get EndFrames(), NewItem_Value:int=@n19) -> (then, ReturnValue)
   n22 = Get ErrorTxt()
   SetVisibility(execute=@n6.CastFailed, self=@n22, InVisibility=Visible)
   > Set Text(Text="LOCTEXT(\"Archetype Leader AnimSET not found \\r\\nSome anim seq not generated\")", self=@n22)
   > Set StartFrames(StartFrames:Array of Frame Number Structures=Make Array!("[0]_Value":int=0, "[1]_Value":int=35) -> (Array:Array of Frame Number Structures)) -> (Output_Get:Array of Frame Number Structures)
   > Set EndFrames(EndFrames:Array of Frame Number Structures=Make Array("[0]_Value"=35, "[1]_Value"=75)) -> (Output_Get:Array of Frame Number Structures)
   > n23 = Set MarkerLabels(MarkerLabels:Array of Strings=Make Array!("[0]":String="##In", "[1]":String="##Out", "[2]":String="##") -> (Array:Array of Strings)) -> (Output_Get:Array of Strings)
   > n24 = For Each Loop(Array=@n23)
   > Add Marked Frame(
    Sequence=Get CurrentLvlSeq()
    InMarkedFrame_FrameNumber_Value="Get (a ref)"!(Array=Get Markers() -> (Markers:Array of Frame Number Structures), Dimension 1=@n24.Array Index) -> (Output_Value:int)
    InMarkedFrame_Label=@n24
    InMarkedFrame_bIsDeterminismFence=False
    InMarkedFrame_bIsInclusiveTime=False
    TimeUnit=DisplayRate
   ) -> (ReturnValue=0)
  ]}
 CreateBlendSpaces {nodes [
   CreateBlendSpaces() {is=title K2Node_FunctionEntry}
   > n1 = For Each Loop(Array:Array of Widget Object References=GetAllChildren(self=Get BlendSpaceColumnHolder() -> (BlendSpaceColumnHolder:Stack Box Object Reference))) -> (Array Element:Widget Object Reference)
   > n2 = Cast To AnimLists!(execute, Object=@n1) -> (then, CastFailed, AsAnim Lists:Anim Lists Object Reference) {is=title K2Node_DynamicCast}
   > n3 = For Each Loop(Array=GetAllChildren(self=Get RowHolder!(self:Anim Lists Object Reference=@n2) -> (RowHolder:Stack Box Object Reference)))
   > n4 = Cast To AnimListItem!(execute, Object=@n3) -> (then, CastFailed, AsAnim List Item:Anim List Item Object Reference) {is=title K2Node_DynamicCast}
   > Branch(Condition=IsChecked(self=Get IncludeCheck!(self:Anim List Item Object Reference=@n4) -> (IncludeCheck:Editor Utility Check Box Object Reference)))
   > K2Node_Composite_0 = Set Data!(execute, self:Anim List Item Object Reference=@n4) -> (then:Exec) {is=title K2Node_Composite}
   > GeneratePaths()
   > n5 = "For Each Loop (Map)"(MapPin:Map of Strings to Vectors=Get Anim() -> (Anim:Map of Strings to Vectors)) -> (ValuePin:Vector)
   n6 = Get SelectedAC()
   n7 = Find(
    TargetMap:Map of Strings to Anim Sequence Object References=Get AnimCollection()
    Key=Append!(A=Get Asset(self=@n6), B=_, C=@n5) -> (ReturnValue)
   ) -> (Value:Anim Sequence Object Reference)
   > Branch(Condition=@n7.ReturnValue)
   > Add!(
    execute
    TargetArray:Array of Anim Sequence Object References=Get BlendSpaceAnims() -> (BlendSpaceAnims:Array of Anim Sequence Object References)
    NewItem:"Anim Sequence Object Reference (by ref)"=@n7
   ) -> (then, ReturnValue)
   n8 = Switch on ENUM_BlendSpaceType!(
    execute=@n5.CompletedPin
    Selection:ENUM_BlendSpaceType Enum=Get BlendSpaceType() -> (BlendSpaceType:ENUM_BlendSpaceType Enum=NewEnumerator0)
   ) -> (NewEnumerator0:Exec, NewEnumerator1:Exec, NewEnumerator2:Exec) {is=title K2Node_SwitchEnum}
   > n9 = Set AssetClass(AssetClass:Blend Space Class Reference=/Script/Engine.AimOffsetBlendSpace) -> (Output_Get:Blend Space Class Reference)
   n10 = Get BlendSpaceAnims()
   n11 = Get Anim()
   Branch(
    execute=[@n21.then, @n20.then, @n9.then]
    Condition="Equal (Integer)"(
     A=Length(TargetArray:Array of Anim Sequence Object References=@n10)
     B=Length!(TargetMap:Map of Strings to Vectors=@n11) -> (ReturnValue:int=0) {is=title K2Node_CallFunction}
    )
   )
   n12 = "Get (a copy)"(Array:Array of Anim Sequence Object References=@n10) -> (Output:Anim Sequence Object Reference)
   > Set SeqRef(SeqRef:Anim Sequence Object Reference=@n12) -> (Output_Get:Anim Sequence Object Reference)
   n13 = Get AnimSeqPath()
   > n14 = Execute Python Script!(
    execute
    PythonScript="import unreal\r\n\r\nasset_tools = unreal.AssetToolsHelpers.get_asset_tools()\r\n\r\npackage_path = package_path\r\nasset_name = asset_name\r\n\r\naim_offset_factory = unreal.BlendSpaceFactoryNew()\r\nskeleton = skeleton\r\naim_offset_factory.target_skeleton = skeleton\r\n\r\nnew_asset = asset_tools.create_asset(\r\n    asset_name,\r\n    package_path,\r\n    asset_class,\r\n    aim_offset_factory\r\n)"
    skeleton=GetSkeleton(InAsset=@n12)
    asset_name=GetBaseFilename(InPath=@n13, bRemovePath=true)
    package_path=GetPath(InPath=@n13)
    asset_class:Blend Space Class Reference=Get AssetClass() -> (AssetClass:Blend Space Class Reference)
   ) -> (then, ReturnValue, new_asset:Blend Space Object Reference) {inputs [skeleton,asset_name,package_path,asset_class], outputs [new_asset]}
   > Set BlendSpaceRef(BlendSpaceRef:Blend Space Object Reference=@n14.new_asset) -> (Output_Get:Blend Space Object Reference)
   n15 = Get!() -> (
    BlendSpaceSettings_HorizontalName_11_#1:String
    BlendSpaceSettings_HorizontalMin_5_#2:double=~
    BlendSpaceSettings_HorizontilMax_7_#3:double=~
    BlendSpaceSettings_VerticalName_12_#4:String
    BlendSpaceSettings_VerticalMin_15_#5:double=~
    BlendSpaceSettings_VerticalMax_17_#6:double=~
    BlendSpaceSettings_AdditiveBasePreviewPoseName_22_#7:String
    BlendSpaceSettings_SmoothTime_25_#8:double=~
   )
   n16 = Get BlendSpaceRef() -> (BlendSpaceRef:Blend Space Object Reference)
   > Execute Python Script!(
    execute
    PythonScript="import unreal\r\n\r\nblend_space_path =bs_path\r\n       \r\n## Load BlendSpace asset\r\nblend_space_asset = bs\r\nif not blend_space_asset:\r\n    raise RuntimeError(f\"Failed to load BlendSpace at: {blend_space_path}\")\r\n\r\nparams = blend_space_asset.get_editor_property(\"blend_parameters\")\r\n\r\nhorizontal = params[0]\r\nvertical = params[1]\r\n\r\nhorizontal.set_editor_property(\"display_name\", horizontal_name)\r\nhorizontal.set_editor_property(\"min\", horizontal_min)\r\nhorizontal.set_editor_property(\"max\", horizontal_max)\r\n\r\nvertical.set_editor_property(\"display_name\", vertical_name)\r\nvertical.set_editor_property(\"min\", vertical_min)\r\nvertical.set_editor_property(\"max\", vertical_max)\r\n\r\n## Put modified struct back\r\nparams[0] = horizontal\r\nbs.set_editor_property(\"blend_parameters\", params)\r\nparams[1] = vertical\r\nbs.set_editor_property(\"blend_parameters\", params)\r\nbs.set_editor_property(\"Preview_Base_Pose\", zero_pose)"
    bs:Blend Space Object Reference=@n16
    bs_path:String=Get Object Path String(Object=@n16)
    zero_pose:Anim Sequence Object Reference=GetAnim!(AnimName:String=@n15.BlendSpaceSettings_AdditiveBasePreviewPoseName_22_#7) -> (NewParam:Anim Sequence Object Reference)
    horizontal_name:String=@n15
    horizontal_min:double=@n15.BlendSpaceSettings_HorizontalMin_5_#2
    horizontal_max:double=@n15.BlendSpaceSettings_HorizontilMax_7_#3
    vertical_name:String=@n15.BlendSpaceSettings_VerticalName_12_#4
    vertical_min:double=@n15.BlendSpaceSettings_VerticalMin_15_#5
    vertical_max:double=@n15.BlendSpaceSettings_VerticalMax_17_#6
   ) -> (then, ReturnValue) {inputs [bs, bs_path, zero_pose, horizontal_name, horizontal_min, horizontal_max, vertical_name, vertical_min, vertical_max]}
   n17 = Get BlendSpaceAnims()
   > For Each Loop(Array:Array of Anim Sequence Object References=@n17) -> (Array Element:Anim Sequence Object Reference)
   > n18 = Values!(execute, TargetMap:Map of Strings to Vectors=Get Anim()) -> (then, Values:Array of Vectors)
   n19 = Get BlendSpaceRef()
   > Execute Python Script!(
    execute
    PythonScript="import unreal\r\n\r\nblend_space_path = bs_path\r\n\r\n## Animation asset paths (your array)\r\nanim_sequences = anim\r\n\r\n\r\ncoordinates = coordinate\r\nblend_space = bs\r\nif not blend_space:\r\n    raise RuntimeError(f\"Could not load Blend Space: {blend_space_path}\")\r\n\r\n\r\nsample_data = blend_space.get_editor_property(\"sample_data\")\r\n\r\nfor anim, pos in zip(anim_sequences, coordinates):\r\n    sample = unreal.BlendSample()\r\n    sample.set_editor_property(\"animation\", anim)\r\n\r\n    ## Directly assign the Vector from the array\r\n    sample.set_editor_property(\"sample_value\", pos)\r\n\r\n    sample_data.append(sample)\r\n\r\n   \r\n## write back\r\nblend_space.set_editor_property(\"sample_data\", sample_data)\r\nunreal.EditorAssetLibrary.save_loaded_asset(blend_space)\r\n"
    bs=@n19
    bs_path=Get Object Path String(Object=@n19)
    anim:Array of Anim Sequence Object References=@n17
    coordinate:Array of Vectors=@n18
   ) -> (then, ReturnValue) {inputs [bs,bs_path,anim,coordinate]}
   > Clear(TargetArray:Array of Anim Sequence Object References=Get BlendSpaceAnims())
   > Clear!(execute, TargetMap:Map of Strings to Vectors=Get Anim()) -> (then) {is=title K2Node_CallFunction}
   n20 = Set AssetClass(execute=@n8.NewEnumerator1, AssetClass=/Script/Engine.BlendSpace)
   n21 = Set AssetClass(execute=@n8.NewEnumerator2, AssetClass=/Script/Engine.BlendSpace1D)
  ]}
 Set Data {
  path=CreateBlendSpaces.K2Node_Composite_0.Set Data
  nodes [
   n1 = Inputs!() -> (execute, self:Anim List Item Object Reference)
   n2 = "To String (Text)"(InText=Get Text(self:Text Object Reference=Get AnimNameTxt!(self:Anim List Item Object Reference=@n1) -> (AnimNameTxt:Text Object Reference)))
   > Set CurrentAnimName(CurrentAnimName:String=@n2) -> (Output_Get:String)
   > n3 = Get Data Table Row DT_BlendSpaceSettings!(
    execute:Exec
    DataTable:Data Table Object Reference=/AnimInEdTools/CompanionAnimAssetGenerator/Settings/DT_BlendSpaceSettings
    RowName:Name=String To Name(InString=@n2)
   ) -> (
    then:Exec
    RowNotFound:Exec
    ReturnValue_BlendSpaceType_17_#9:ENUM_BlendSpaceType Enum
    ReturnValue_Subfolder_56_#10:String
    ReturnValue_BlendSpaceSettings_57_#11:Struct Blend Space Settings Structure
    ReturnValue_AnimandCoordinates_61_#12:Map of Strings to Vectors
   ) {is=title K2Node_GetDataTableRow}
   > Set CurrentSubfolder(CurrentSubfolder:String=@n3.ReturnValue_Subfolder_56_#10) -> (Output_Get:String)
   > Set BlendSpaceType(BlendSpaceType:ENUM_BlendSpaceType Enum=@n3) -> (Output_Get:ENUM_BlendSpaceType Enum=NewEnumerator0)
   > Set BlendSpaceSettings(BlendSpaceSettings:Struct Blend Space Settings Structure=@n3.ReturnValue_BlendSpaceSettings_57_#11) -> (Output_Get:Struct Blend Space Settings Structure)
   > n4 = "For Each Loop (Map)"(MapPin=@n3.ReturnValue_AnimandCoordinates_61_#12)
   n5 = Get Archetype!(self:Animatable Creature Object Reference=Get SelectedAC()) -> (Archetype:ENUM_CompanionArchetype Enum=NewEnumerator0)
   n6 = "Equal (Enum)"(A:ENUM_CompanionArchetype Enum=@n5, B:ENUM_CompanionArchetype Enum=NewEnumerator1)
   n7 = "Equal (Enum)"(A=@n5, B=NewEnumerator0)
   > n8 = Branch(Condition=AND Boolean(A=Contains(SearchIn=@n4, Substring=45), B=OR Boolean!(A=@n7, B=@n6) -> (ReturnValue)))
   Add!(execute=@n8.else, TargetMap:Map of Strings to Vectors=Get Anim(), Key:"String (by ref)"=@n4, Value:"Vector (by ref)"=@n4.ValuePin) -> (then) {is=title K2Node_CallFunction}
   Outputs!(then:Exec=@n4.CompletedPin)
   GetDataTableRowNames()
  ]
 }
 CreateDataAsset {nodes [
   CreateDataAsset() {is=title K2Node_FunctionEntry}
   > Branch(Condition=IsChecked(self=Get DataAssetCheckBox() -> (DataAssetCheckBox:Editor Utility Check Box Object Reference)))
   n1 = Append!(A=GetArchetypeLeaderName(), B=_DA) -> (ReturnValue)
   n2 = Append(A=GetArchetypeLeaderAssetPath(), B=/Blueprint/)
   > n3 = LoadAsset(AssetPath=Append!(A=@n2, B=@n1, C:String=., D:String=@n1) -> (ReturnValue))
   > Is Valid(InputObject=@n3)
   n4 = GetArchetypeLeaderName()
   n5 = Get Asset(self=Get SelectedAC())
   > n6 = DuplicateLoadedAsset!(
    execute
    SourceAsset:Object Reference=@n3
    DestinationAssetPath:String=Append!(A=Replace(SourceString=@n2, From=@n4, To=@n5), B=@n1) -> (ReturnValue)
   ) -> (then, ReturnValue:Object Reference)
   > n7 = Set DuplicatedAsset(DuplicatedAsset:Object Reference=@n6) -> (Output_Get:Object Reference)
   > RenameAsset!(execute, Asset:Object Reference=@n7, NewName:String=RenameDuplicate!(Object:Object Reference=@n3) -> (ReturnValue:String)) -> (then)
   > n8 = Switch on ENUM_CompanionArchetype!(execute, Selection:ENUM_CompanionArchetype Enum=Get Archetype(self=Get SelectedAC()) -> (Archetype=NewEnumerator0)) -> (
    NewEnumerator0:Exec
    NewEnumerator1:Exec
    NewEnumerator7:Exec
    NewEnumerator2:Exec
    NewEnumerator3:Exec
    NewEnumerator4:Exec
    NewEnumerator5:Exec
    NewEnumerator6:Exec
    NewEnumerator8:Exec
   ) {is=title K2Node_SwitchEnum}
   n9 = Get DuplicatedAsset() -> (DuplicatedAsset:Object Reference)
   n10 = Cast To FortUserDefinedDataAsset!(execute=[@n8.NewEnumerator0, @n8.NewEnumerator1], Object=@n9) -> (then, CastFailed, AsFort User Defined Data Asset:Fort User Defined Data Asset Object Reference) {is=title K2Node_DynamicCast}
   > K2Node_Composite_0 = SetQuadrupedDA!(execute, DataAsset:Fort User Defined Data Asset Object Reference=@n10) {is=title K2Node_Composite}
   n11 = Cast To FortUserDefinedDataAsset(execute=[@n8.NewEnumerator2, @n8.NewEnumerator3], Object=@n9)
   > K2Node_Composite_1 = SetRaptorDA!(execute, InputPin:Fort User Defined Data Asset Object Reference=@n11) {is=title K2Node_Composite}
   n12 = Cast To FortUserDefinedDataAsset(execute=[@n8.NewEnumerator4, @n8.NewEnumerator5], Object=@n9)
   > K2Node_Composite_2 = SetBipedDA!(execute, InputPin:Fort User Defined Data Asset Object Reference=@n12) {is=title K2Node_Composite}
   n13 = Cast To FortUserDefinedDataAsset(execute=@n8.NewEnumerator6, Object=@n9)
   > K2Node_Composite_3 = Other/Floater!(execute, InputPin:Fort User Defined Data Asset Object Reference=@n13) {is=title K2Node_Composite}
  ]}
 SetQuadrupedDA {
  path=CreateDataAsset.K2Node_Composite_0.SetQuadrupedDA
  nodes [
   n1 = Inputs!() -> (execute, DataAsset:Fort User Defined Data Asset Object Reference)
   > n2 = Set members in Quadruped Base Struct!(
    execute:Exec
    StructRef:"Quadruped Base Struct Structure (by ref)"=Get QuadrupedStruct() -> (QuadrupedStruct:Quadruped Base Struct Structure)
    LocomotionStart_4_#13:Anim Sequence Object Reference=GetDataAssetAnim!(AnimName:String=COR_Loco_Start) -> (NewParam:Anim Sequence Object Reference)
    LocomotionStop_Slow_6_#14:Anim Sequence Object Reference=GetDataAssetAnim(AnimName=COR_Loco_StopLight)
    LocomotionStop_Fast_8_#15:Anim Sequence Object Reference=GetDataAssetAnim(AnimName=COR_Loco_StopHeavy)
    IdlePose_10_#16:Anim Sequence Object Reference=GetDataAssetAnim(AnimName=COR_Stand_Pose)
    IdleAngry_Loop_12_#17:Anim Sequence Object Reference=GetDataAssetAnim(AnimName=COR_Angry_Idle_Lp)
    BodyTurn_BS_51_#18:Blend Space 1D Object Reference=GetBlendSpace1D!(AnimName:String=Loco_Turn_BS) -> (NewParam:Blend Space 1D Object Reference)
    Breathing_Loop_16_#19:Anim Sequence Object Reference=GetDataAssetAnim(AnimName=COR_Breathing_ADD)
    Locomotion_Sprint_18_#20:Anim Sequence Object Reference=GetDataAssetAnim(AnimName=COR_Loco_Sprint)
    Locomotion_Run_20_#21:Anim Sequence Object Reference=GetDataAssetAnim(AnimName=COR_Loco_Run)
    Locomotion_Trot_22_#22:Anim Sequence Object Reference=GetDataAssetAnim(AnimName=COR_Loco_Trot)
    Locomotion_Walk_24_#23:Anim Sequence Object Reference=GetDataAssetAnim(AnimName=COR_Loco_Walk)
    Tall_Wag_BS_32_#24:Blend Space 1D Object Reference=GetBlendSpace1D(AnimName=Tail_BS)
    TallWag_Sit_28_#25:Anim Sequence Object Reference=GetDataAssetAnim(AnimName=COR_Tail_Sit_ADD)
    Tail_Wag_Angry_30_#26:Anim Sequence Object Reference=GetDataAssetAnim(AnimName=COR_Tail_Angry_ADD)
    BodyPose_Worried_35_#27:Anim Sequence Object Reference=GetDataAssetAnim(AnimName=COR_Loco_Worried_ADD)
    BodyPose_Combat_37_#28:Anim Sequence Object Reference=GetDataAssetAnim(AnimName=COR_Loco_Angry_ADD)
    SittingToStand_39_#29:Anim Sequence Object Reference=GetDataAssetAnim(AnimName=COR_Sit_Out)
    Sitting_41_#30:Anim Sequence Object Reference=GetDataAssetAnim(AnimName=COR_Sit_Pose)
    StandToSitting_43_#31:Anim Sequence Object Reference=GetDataAssetAnim(AnimName=COR_Sit_In)
    AngryIdle_Intro_54_#32:Anim Sequence Object Reference=GetDataAssetAnim(AnimName=COR_Angry_Idle_In)
    AngryIdle_Exit_45_#33:Anim Sequence Object Reference=GetDataAssetAnim(AnimName=COR_Angry_Idle_Out)
    IdleAntic_47_#34:Anim Sequence Object Reference=GetDataAssetAnim(AnimName=ACT_Break_ADD)
    AimOffset_50_#35:Blend Space Object Reference=GetBlendSpace1D(AnimName=AO)
    Chonky_Sit_ADD_60_#36:Anim Sequence Object Reference=GetDataAssetAnim(AnimName=COR_Sit_Chonky_ADD)
   ) -> (then:Exec, StructOut:"Quadruped Base Struct Structure (by ref)") {is=title K2Node_SetFieldsInStruct}
   > n3 = MakeInstancedStruct!(execute:Exec, Value:"Quadruped Base Struct Structure (by ref)"=@n2) -> (then:Exec, ReturnValue:Instanced Struct Structure) {is=title K2Node_InstancedStruct}
   > Set UserDefinedStruct!(execute, UserDefinedStruct:Instanced Struct Structure=@n3, self:Fort User Defined Data Asset Object Reference=@n1) -> (then, Output_Get:Instanced Struct Structure)
   Outputs!()
  ]
 }
 SetRaptorDA {
  path=CreateDataAsset.K2Node_Composite_1.SetRaptorDA
  nodes [
   n1 = Inputs!() -> (execute, InputPin:Fort User Defined Data Asset Object Reference)
   > n2 = Set members in Biped Raptor Base Struct!(
    execute:Exec
    StructRef:"Biped Raptor Base Struct Structure (by ref)"=Get RaptorStruct() -> (RaptorStruct:Biped Raptor Base Struct Structure)
    Locomotion_Start_98_#37:Anim Sequence Object Reference=GetAnim(AnimName=COR_Loco_Start)
    Locomotion_Idle_72_#13:Anim Sequence Object Reference=GetAnim(AnimName=COR_Idle)
    Locomotion_Sprint_73_#38:Anim Sequence Object Reference=GetAnim(AnimName=COR_Loco_Sprint)
    Locomotion_Run_74_#39:Anim Sequence Object Reference=GetAnim(AnimName=COR_Loco_Run)
    Locomotion_Jog_75_#40:Anim Sequence Object Reference=GetAnim(AnimName=COR_Loco_Jog)
    Locomotion_Walk_76_#41:Anim Sequence Object Reference=GetAnim(AnimName=COR_Loco_Walk)
    BodyBend_BS_68_#42:Blend Space 1D Object Reference=GetBlendSpace1D(AnimName=Body_BS)
    BodyPose_Worried_77_#43:Anim Sequence Object Reference=GetAnim(AnimName=COR_Worried_ADD)
    AimOffset_80_#44:Blend Space Object Reference=GetAO!(AnimName:String=_AO) -> (NewParam:Aim Offset Blend Space Object Reference)
    StandToSit_89_#45:Anim Sequence Object Reference=GetAnim(AnimName=COR_Sit_In)
    Sitting_88_#46:Anim Sequence Object Reference=GetAnim(AnimName=COR_Sit_Lp)
    SitToStand_87_#47:Anim Sequence Object Reference=GetAnim(AnimName=COR_Sit_Out)
    TailBend_BS_92_#48:Blend Space 1D Object Reference=GetBlendSpace1D(AnimName=Tail_BS)
    FeatherSpreadPose_96_#49:Anim Sequence Object Reference=GetAnim(AnimName=QRK_FeatherSplay)
   ) -> (then:Exec, StructOut:"Biped Raptor Base Struct Structure (by ref)") {is=title K2Node_SetFieldsInStruct}
   > n3 = MakeInstancedStruct(Value:"Biped Raptor Base Struct Structure (by ref)"=@n2)
   > Set UserDefinedStruct(UserDefinedStruct=@n3, self=@n1)
   Outputs()
  ]
 }
 SetBipedDA {
  path=CreateDataAsset.K2Node_Composite_2.SetBipedDA
  nodes [
   n1 = Inputs()
   > n2 = Set members in Biped Base Struct!(
    execute:Exec
    StructRef:"Biped Base Struct Structure (by ref)"=Get BipedStruct() -> (BipedStruct:Biped Base Struct Structure)
    Locomotion_Idle_72_#13:Anim Sequence Object Reference=GetAnim(AnimName=COR_Stand_Lp)
    Locomotion_Sprint_73_#38:Anim Sequence Object Reference=GetAnim(AnimName=COR_Loco_Sprint)
    Locomotion_Run_74_#39:Anim Sequence Object Reference=GetAnim(AnimName=COR_Loco_Run)
    Locomotion_Trot_99_#40:Anim Sequence Object Reference=GetAnim(AnimName=COR_Loco_Trot)
    Body_Turn_BS_100_#42:Blend Space 1D Object Reference=GetBlendSpace1D(AnimName=Body_Turn_BS)
    AimOffset_80_#44:Blend Space Object Reference=GetAO(AnimName=_AO)
    Breathing_102_#43:Anim Sequence Object Reference=GetAnim(AnimName=COR_Breathing_ADD)
    Worried_Pose_103_#46:Anim Sequence Object Reference=GetAnim(AnimName=COR_Worried_Pose_ADD)
    Worried_Locomotion_Pose_104_#47:Anim Sequence Object Reference=GetAnim(AnimName=COR_Loco_Worried_ADD)
    Bored_Pose_Additive_124_#50:Anim Sequence Object Reference=GetAnim(AnimName=COR_Bored_Pose_ADD)
    Bored_Intro_125_#51:Anim Sequence Object Reference=GetAnim(AnimName=COR_Bored_In)
    Bored_Outro_126_#52:Anim Sequence Object Reference=GetAnim(AnimName=COR_Bored_Out)
    Angry_Locomotion_Pose_127_#53:Anim Sequence Object Reference=GetAnim(AnimName=COR_Loco_Angry_ADD)
    Angry_Idle_Intro_128_#54:Anim Sequence Object Reference=GetAnim(AnimName=COR_Angry_Idle_In)
    Angry_Idle_Loop_129_#55:Anim Sequence Object Reference=GetAnim(AnimName=COR_Angry_Idle_Lp)
    Angry_Idle_Outro_130_#56:Anim Sequence Object Reference=GetAnim(AnimName=COR_Angry_Idle_Out)
    Zero_Pose_136_#57:Anim Sequence Object Reference=GetAnim(AnimName=COR_Base_Pose)
    Tail_Pose_133_#58:Anim Sequence Object Reference=GetAnim(AnimName=COR_Tail_Pose_ADD)
   ) -> (then:Exec, StructOut:"Biped Base Struct Structure (by ref)") {is=title K2Node_SetFieldsInStruct}
   > n3 = MakeInstancedStruct(Value:"Biped Base Struct Structure (by ref)"=@n2)
   > Set UserDefinedStruct(UserDefinedStruct=@n3, self=@n1)
   GetAnim(AnimName=WCOR_Loco_Walk)
   Outputs()
  ]
 }
 Other/Floater {
  path=CreateDataAsset.K2Node_Composite_3.Other/Floater
  nodes [
   n1 = Inputs()
   > n2 = Set members in Floater Base Struct!(
    execute:Exec
    StructRef:"Floater Base Struct Structure (by ref)"=Get OtherStruct() -> (OtherStruct:Floater Base Struct Structure)
    Loco_Start_Slow_84_#14:Anim Sequence Object Reference=GetAnim(AnimName=COR_Loco_Start)
    Loco_Start_Fast_83_#15:Anim Sequence Object Reference=GetAnim(AnimName=COR_Loco_StartFast)
    Idle_Loop_65_#16:Anim Sequence Object Reference=GetAnim(AnimName=COR_Idle_Lp)
    Body_Turn_BS_80_#18:Blend Space 1D Object Reference=GetBlendSpace1D(AnimName=COR_Loco_Turn_BS)
    Locomotion_BS_91_#20:Blend Space 1D Object Reference=GetBlendSpace1D(AnimName=COR_Loco_Swim_BS)
    Bubble_Loop_ADD_79_#25:Anim Sequence Object Reference=GetAnim(AnimName=COR_IdleBubble_Lp_ADD)
    Bubble_Stretch_BS_92_#59:Blend Space 1D Object Reference=GetBlendSpace1D(AnimName=COR_Loco_Bubble_BS)
    Pose_Base_74_#26:Anim Sequence Object Reference=GetAnim(AnimName=COR_Base_Pose)
    Pose_Worried_ADD_78_#27:Anim Sequence Object Reference=GetAnim(AnimName=COR_Worried_ADD)
    Bored_Intro_70_#29:Anim Sequence Object Reference=GetAnim(AnimName=COR_Bored_In)
    Bored_Loop_71_#30:Anim Sequence Object Reference=GetAnim(AnimName=COR_Bored_Lp)
    Bored_Outro_72_#31:Anim Sequence Object Reference=GetAnim(AnimName=COR_Bored_Out)
    Angry_Idle_Loop_76_#17:Anim Sequence Object Reference=GetAnim(AnimName=COR_Angry_Idle_Lp)
    Angry_Idle_Outro_69_#33:Anim Sequence Object Reference=GetAnim(AnimName=COR_Angry_Idle_Out)
    AimOffset_93_#35:Aim Offset Blend Space Object Reference=GetAO(AnimName=_AO)
    Child_Override_ADD_63_#36:Anim Sequence Object Reference=GetAnim(AnimName=ACT_WingFlapA)
   ) -> (then:Exec, StructOut:"Floater Base Struct Structure (by ref)") {is=title K2Node_SetFieldsInStruct}
   > n3 = MakeInstancedStruct(Value:"Floater Base Struct Structure (by ref)"=@n2)
   > Set UserDefinedStruct(UserDefinedStruct=@n3, self=@n1)
   GetBlendSpace1D(AnimName=Tail_BS)
   GetAnim(AnimName=COR_Angry_Idle_In)
   Outputs()
  ]
 }
 CreateMontage {nodes [
   n1 = CreateMontage!() -> (then, AnimSequence:Anim Sequence Object Reference, ObjectPath:String) {is=title K2Node_FunctionEntry}
   > n2 = Construct Object from Class(Class=/AnimInEdTools/AnimationBrowser/Dialog/CreateAnimMontage/MontagePreset/BlankMontage.BlankMontage_C) -> (ReturnValue:Blank Montage Object Reference)
   n3 = GetBaseFilename(InPath=@n1.ObjectPath, bRemovePath=true)
   > n4 = Create Montage!(
    execute
    self:Blank Montage Object Reference=@n2
    AssetName:String=Append(A=Replace(SourceString=@n3, From=_in), B=_M)
    PackagePath:String=GetPath(InPath=@n1.ObjectPath)
    AnimSequence:Anim Sequence Object Reference=@n1
   ) -> (then, AnimMontage:Anim Montage Object Reference)
   > n5 = Branch(Condition=Contains(SearchIn=GetDisplayName(Object=@n4), Substring=Interact))
   n6 = Get SlotAnimTracks!(self:Anim Montage Object Reference=@n4) -> (SlotAnimTracks:Array of Slot Animation Track Structures)
   > n7 = AddAnimationNotifyStateEvent!(
    execute
    AnimationSequenceBase:Anim Sequence Base Object Reference=@n4
    NotifyTrackName:Name=1
    StartTime:float=0
    Duration:float="Get (a ref)"!(
     Array:Array of Anim Segment Structures="Get (a ref)"!(Array:Array of Slot Animation Track Structures=@n6, Dimension 1=0) -> (Output_SlotName:Name, Output_AnimTrack_AnimSegments:Array of Anim Segment Structures).Output_AnimTrack_AnimSegments
     Dimension 1=0
    ) -> (
     Output_AnimReference:Anim Sequence Base Object Reference
     Output_StartPos:float=~
     Output_AnimStartTime:float=~
     Output_AnimEndTime:float=~
     Output_AnimPlayRate:float=~
     Output_LoopingCount:int
    ).Output_AnimEndTime
    NotifyStateClass:Anim Notify State Class Reference=/CosmeticCompanions/Assets/Shared/AnimNotifies/ANS_Companion_UpdateInteractionIK.ANS_Companion_UpdateInteractionIK_C
   ) -> (then, ReturnValue:Anim Notify State Object Reference)
   SaveLoadedAssets(
    execute=[@n7.then, @n5.else]
    AssetsToSave=Make Array!("[0]":Object Reference=@n4) -> (Array:Array of Object References)
    bOnlyIfIsDirty=true
   )
  ]}
 EventGraph {nodes [
   "On Clicked (GenerateBtn)"() {is=title K2Node_ComponentBoundEvent}
   > GenerateSequences()
   > SetAdditiveSettings()
   > MiscAnimSettings()
   > CheckNeedsMontage() {is=title K2Node_CallFunction}
   > CreateBlendSpaces() {is=title K2Node_CallFunction}
   > GenerateChooserTables()
   > CreateDataAsset() {is=title K2Node_CallFunction}
   > GenerateAnimBP()
   > GenerateCAS()
   > SaveAllAssets()
   n1 = "On Selection Changed (Asset_ComboBox)"!() -> (OutputDelegate, then, SelectedItem:String, SelectionType:ESelectInfo Enum) {is=title K2Node_ComponentBoundEvent}
   > GetSelectedAC!(execute, self, InputPin:String=@n1.SelectedItem) -> (then)
   > SetColumns()
   > SetCheckedState!(
    execute
    self:Check Box Object Reference=Get SelectAllCheckBox() -> (SelectAllCheckBox:Editor Utility Check Box Object Reference)
    InCheckedState:ECheckBoxState Enum=Unchecked
   ) -> (then)
   > uncheck()
   > SetVisibility(self=Get ErrorTxt(), InVisibility=Hidden)
   n2 = "On Check State Changed (SelectAllCheckBox)"!() -> (OutputDelegate, then, bIsChecked:bool) {is=title K2Node_ComponentBoundEvent}
   HandleCheckboxes!(execute=[@n2.then, @n3.then], self, InputPin:bool=IsChecked(self=Get SelectAllCheckBox())) -> (then)
   n3 = uncheck() {is=title K2Node_CustomEvent}
   Event Construct()
   n4 = Get Asset_ComboBox() -> (Asset_ComboBox:"Searchable Combo Box (String) Object Reference")
   > UpdateComboBox!(
    execute
    self
    ComboBox:"Searchable Combo Box (String) Object Reference"=@n4
    DefaultItem:String
    Items:Array of Strings=GetItems!() -> (Items:Array of Strings)
    AddBlank:bool=false
   ) -> (then)
   > SetSelectedIndex!(execute, self:"Searchable Combo Box (String) Object Reference"=@n4, Index:int=0) -> (then)
   Event PreConstruct()
   Event Tick()
  ]}
 GenerateAnimBP {nodes [
   GenerateAnimBP() {is=title K2Node_FunctionEntry}
   > Branch(Condition=IsChecked(self=Get AnimBPCheckBox() -> (AnimBPCheckBox:Editor Utility Check Box Object Reference)))
   n1 = Append(A=GetArchetypeLeaderName(), B=_Base_AnimBP_Child)
   n2 = Append(A=GetArchetypeLeaderAssetPath(), B=/Blueprint/)
   > n3 = LoadAsset(AssetPath=Append!(A=@n2, B=@n1, C:String=., D:String=@n1) -> (ReturnValue))
   > Is Valid(InputObject=@n3)
   n4 = GetArchetypeLeaderName()
   n5 = Get Asset(self=Get SelectedAC())
   > n6 = DuplicateLoadedAsset(SourceAsset=@n3, DestinationAssetPath=Append!(A=Replace(SourceString=@n2, From=@n4, To=@n5), B=@n1) -> (ReturnValue))
   > Is Valid(InputObject=@n6)
   > n7 = Set DuplicatedAsset(DuplicatedAsset=@n6)
   > RenameAsset(Asset=@n7, NewName=RenameDuplicate(Object=@n3))
   > Cast To AnimInstance!(execute, Object=@n7) -> (then, CastFailed, AsAnim Instance:Anim Instance Object Reference) {is=title K2Node_DynamicCast}
  ]}
 GenerateCAS {nodes [
   GenerateCAS() {is=title K2Node_FunctionEntry}
   > Branch(Condition=IsChecked(self=Get CASCheckBox() -> (CASCheckBox:Editor Utility Check Box Object Reference)))
   > n1 = LoadAsset(AssetPath=/CosmeticCompanions/Assets/CAS/CAS_FullSnore_Interact)
   > Is Valid(InputObject=@n1)
   n2 = Get SelectedAC()
   > n3 = DuplicateLoadedAsset(
    SourceAsset=@n1
    DestinationAssetPath=Replace(SourceString=&CAS_FullSnore_Interact, From=FullSnore, To=Get Asset(self=@n2))
   )
   > Is Valid(InputObject=@n3)
   > SaveLoadedAssets(AssetsToSave=Make Array("[0]"=@n3), bOnlyIfIsDirty=true)
  ]}
 GenerateChooserTables {nodes [
   n1 = LoadAsset(AssetPath=/CosmeticCompanions/Assets/Biped/BitterDolly/Animation/Choosers/CHT_BitterDolly_AnticsBored)
   > n2 = LoadAsset(AssetPath=/CosmeticCompanions/Assets/Biped/BitterDolly/Animation/Actions/BitterDolly_Antic_Head_LookCalm)
   > Execute Python Script!(
    execute
    PythonScript="import unreal\r\n\r\nchooser_table = chooser\r\nanim = anim\r\n\r\nresults = chooser_table.set_editor_property(\"results[0]\", anim)\r\n"
    chooser:Object Reference=@n1
    anim:Object Reference=@n2
   ) -> (then, ReturnValue) {inputs [chooser,anim]}
   GenerateChooserTables() {is=title K2Node_FunctionEntry}
   > n3 = For Each Loop(Array:Array of Widget Object References=GetAllChildren(self=Get ChooserTableColumnHolder() -> (ChooserTableColumnHolder:Stack Box Object Reference))) -> (Array Element:Widget Object Reference)
   > n4 = Cast To AnimLists(Object=@n3)
   n5 = Get ColumnName!(self:Anim Lists Object Reference=@n4) -> (ColumnName:Text Object Reference)
   > Set CurrentSubfolder(CurrentSubfolder="To String (Text)"(InText=Get Text(self=@n5)))
   > GeneratePaths()
   > n6 = For Each Loop(Array=GetAllChildren(self=Get RowHolder(self=@n4)))
   > n7 = Cast To AnimListItem(Object=@n6)
   > Branch(Condition=IsChecked(self=Get IncludeCheck(self=@n7)))
   > n8 = Loop Archetype Leader Assets!(execute:Exec, Break:Exec=@n13.then_0, SubFolder:String=/Animation/Choosers/) -> (LoopBody:Exec, AssetPathAndName:String) {is=title K2Node_MacroInstance}
   n9 = GetBaseFilename(InPath=@n8, bRemovePath=true)
   n10 = "To String (Text)"(InText=Get Text(self=Get AnimNameTxt(self=@n7)))
   n11 = Left!(SourceString:String=@n10, Count:int=3) -> (ReturnValue:String)
   n12 = RightChop!(SourceString:String=@n10, Count:int=4) -> (ReturnValue:String)
   > Branch(Condition=AND Boolean(A=Contains(SearchIn=@n9, Substring=@n12), B=Contains(SearchIn=@n9, Substring=@n11)))
   > n13 = Sequence()
   n14 = LoadAsset(execute=@n13.then_1, AssetPath=@n8)
   > Is Valid(InputObject=@n14)
   n15 = Get Class(Object=@n14) -> (ReturnValue:Object Class Reference)
   > Branch(Condition=OR Boolean(A="Equal (Class)"(A=@n15, B=/Script/Chooser.ChooserTable), B="Equal (Class)"(A=@n15, B=/Script/ProxyTable)))
   > n16 = DuplicateLoadedAsset(SourceAsset=@n14, DestinationAssetPath=Get AnimSeqPath())
   > n17 = Set Chooser(Chooser:Object Reference=@n16) -> (Output_Get:Object Reference)
   > RenameAsset(Asset=@n17, NewName=RenameDuplicate(Object=@n14))
   > SaveLoadedAsset(AssetToSave=@n17, bOnlyIfIsDirty=true)
   n18 = "For Each Loop (Map)"(MapPin:Map of Wildcards to Wildcards) -> (KeyPin:Wildcard, ValuePin:Wildcard)
   n19 = Get Chooser() -> (Chooser:Object Reference)
   > Branch(Condition=Contains(SearchIn=GetDisplayName(Object=@n19), Substring=@n18))
   > n20 = For Each Loop(Array:Array of Strings=Reroute Node(InputPin:Array of Strings) -> (OutputPin:Array of Strings)) -> (Array Element:String)
   > n21 = For Each Loop!(Exec, Array:Array of Asset Data Structures=GetAllAnimFolderPaths!() -> (OutAssetData:Array of Asset Data Structures)) -> (
    LoopBody
    Array Element_PackageName:Name
    Array Element_PackagePath:Name
    Array Element_AssetName:Name
    Array Element_AssetClassPath_PathString:String
    Array Index
    Completed
   )
   Execute Python Script!(
    execute=@n21.Completed
    PythonScript="import unreal\r\n\r\ndef replace_chooser_or_proxy_outputs(table, data):\r\n\r\n\r\n    if not table:\r\n        raise RuntimeError(\"Asset is None\")\r\n\r\n    table.modify()\r\n\r\n\r\n    if isinstance(table, unreal.ChooserTable):\r\n        new_results = []\r\n\r\n        for obj in data:\r\n            entry = unreal.ChooserWidgetValueChanged()\r\n\r\n            if isinstance(obj, unreal.Class):\r\n                entry.set_editor_property(\"ResultClass\", obj)\r\n            else:\r\n                entry.set_editor_property(\"ResultObject\", obj)\r\n\r\n            new_results.append(entry)\r\n\r\n\r\n        table.set_editor_property(\"Results\", new_results)\r\n\r\n        unreal.EditorAssetLibrary.save_loaded_table(table)\r\n\r\n        unreal.log(\r\n            f\"Replaced ChooserTable results with {len(new_results)} entries\"\r\n        )\r\n        return\r\n\r\n\r\n    if isinstance(table, unreal.ChooserTableProxy):\r\n        new_values = []\r\n\r\n        for obj in data:\r\n            value = unreal.ChooserTableProxyValue()\r\n            value.set_editor_property(\"Value\", obj)\r\n            value.set_editor_property(\"Name\", obj.get_name())\r\n\r\n            new_values.append(value)\r\n\r\n\r\n        table.set_editor_property(\"Values\", new_values)\r\n\r\n        unreal.EditorAssetLibrary.save_loaded_table(table)\r\n\r\n        unreal.log(\r\n            f\"Replaced ProxyTable values with {len(new_values)} entries\"\r\n        )\r\n        return\r\n\r\n    raise RuntimeError(\r\n        f\"{table.get_name()} is not a ChooserTable or ChooserTableProxy\"\r\n    )\r\n\r\n\r\ntable = table\r\n\r\ndata = data\r\n\r\nreplace_chooser_or_proxy_outputs(table, data)\r\n"
    table:Object Reference=Get Chooser()
    data:Array of Object References=Get DataToSwap() -> (DataToSwap:Array of Object References)
   ) -> (then, ReturnValue) {inputs [table,data]}
   Branch(execute=@n21.LoopBody, Condition=Contains(SearchIn="To String (Name)"(InName=@n21.Array Element_AssetName), Substring=@n20, bUseCase=true))
   > n22 = LoadAsset(AssetPath=Append!(A="To String (Name)"(InName=@n21), B=., C:String="To String (Name)"(InName=@n21.Array Element_AssetName)) -> (ReturnValue))
   > Add(TargetArray:Array of Object References=Get DataToSwap(), NewItem:"Object Reference (by ref)"=@n22) {is=title K2Node_CallArrayFunction}
   Cast To ChooserTable!(execute, Object=@n1) -> (then, CastFailed, AsChooser Table:Chooser Table Object Reference) {is=title K2Node_DynamicCast}
   n23 = GetDataTableRowNames()
   n24 = Enum to String!(Enumerator:ENUM_CompanionArchetype Enum=Get Archetype(self=Get SelectedAC()) -> (Archetype=NewEnumerator0)) -> (ReturnValue:String)
   "Get (a ref)"!(
    Array:Array of Names=@n23
    Dimension 1=Find Item!(TargetArray:Array of Names=@n23, ItemToFind:"Name (by ref)"=String To Name(InString=@n24)) -> (ReturnValue)
   ) -> (Output:"Name (by ref)")
  ]}
 GeneratePaths {nodes [
   GeneratePaths() {is=title K2Node_FunctionEntry}
   n1 = Get SelectedAC()
   n2 = Get Naming Convention!(self:Animatable Creature Object Reference=@n1) -> (ReturnValue:EU Animation Naming Convention Object Reference) {is=title K2Node_CallFunction}
   n3 = Get SelectedAC()
   n4 = Get Path Parts Map!(
    Asset:String=Get Asset(self=@n3)
    Animation Set:String
    Animation Name:String=Get CurrentAnimName() -> (CurrentAnimName:String)
    Descriptor:String
    Direction:String
    Subfolder:String=Get CurrentSubfolder()
    Variation:String
    Plugin:String=Get Variant!(self:Animatable Creature Object Reference=@n3) -> (Variant:String)
    __WorldContext:Object Reference
   ) -> (Output Path Parts Map:Map of Strings to Strings) {is=title K2Node_CallFunction}
   > Set AnimSeqPath(AnimSeqPath:String=Get Path!(
     self:EU Animation Naming Convention Object Reference=@n2
     Token Map:Map of Strings to Strings=@n4
     Path Pattern:String=Get Animation Path!(self:Animatable Creature Object Reference=@n1) -> (Animation Path:String)
     Remove Spaces On Name:bool=false
    ) -> (File Path:String, Folder Path:String, File Name:String) {is=title K2Node_CallFunction}) -> (Output_Get:String)
   > Set LvlSeqPath(LvlSeqPath:String=Get Path(
     self=@n2
     Token Map=@n4
     Path Pattern=Get Source Path!(self:Animatable Creature Object Reference=@n1) -> (Source Path:String)
     Remove Spaces On Name=false
    )) -> (Output_Get:String)
  ]}
 GenerateSequences {nodes [
   GenerateSequences() {is=title K2Node_FunctionEntry}
   > n1 = For Each Loop!(
    Exec
    Array:Array of Widget Object References=GetAllChildren(self=Get SequenceColumnHolder() -> (SequenceColumnHolder:Stack Box Object Reference))
   ) -> (LoopBody, Array Element:Widget Object Reference, Array Index, Completed)
   > n2 = Cast To AnimLists(Object=@n1)
   n3 = Get ColumnName(self=@n2)
   > Set CurrentSubfolder(CurrentSubfolder="To String (Text)"(InText=Get Text(self=@n3)))
   > n4 = For Each Loop(Array=GetAllChildren(self=Get RowHolder(self=@n2)))
   > n5 = Cast To AnimListItem(Object=@n4)
   > Branch(Condition=IsChecked(self=Get IncludeCheck(self=@n5)))
   n6 = Get AnimNameTxt(self=@n5)
   > Set CurrentAnimName(CurrentAnimName="To String (Text)"(InText=Get Text(self=@n6)))
   > GeneratePaths() {is=title K2Node_CallFunction}
   n7 = Get SelectedAC()
   n8 = Get Sequencer Actor(self=@n7)
   > n9 = Branch(Condition=AND Boolean(A=IsValid(Object=@n7), B=IsValidSoftClassReference!(SoftClassReference:"Object Soft Class Reference (by ref)"=@n8) -> (ReturnValue:bool)))
   > SetVisibility(self=Get ErrorTxt(), InVisibility=Hidden)
   n10 = Get LvlSeqPath() -> (LvlSeqPath:String)
   > n11 = Execute Python Script!(
    execute
    PythonScript="import unreal\r\n\r\n## Get asset tools\r\nasset_tools = unreal.AssetToolsHelpers.get_asset_tools()\r\n \r\n## Creates a Level Sequence - /Game/Test_LS\r\nlevel_sequence = unreal.AssetTools.create_asset(asset_tools, asset_name = asset_name, package_path = package_path, asset_class = unreal.LevelSequence, factory = unreal.LevelSequenceFactoryNew())\r\n\r\n## Then open it in sequencer\r\nunreal.LevelSequenceEditorBlueprintLibrary.open_level_sequence(level_sequence)"
    asset_name:String=GetCleanFilename!(InPath:String=@n10) -> (ReturnValue:String)
    package_path:String=GetPath(InPath=@n10)
   ) -> (then, ReturnValue, level_sequence:Object Reference) {inputs [asset_name,package_path], outputs [level_sequence]}
   > Branch(Condition=@n11)
   > n12 = Cast To LevelSequence(Object=@n11.level_sequence)
   > n13 = Set CurrentLvlSeq(CurrentLvlSeq:Level Sequence Object Reference=@n12) -> (Output_Get:Level Sequence Object Reference)
   > SetDisplayRate!(execute, Sequence:Movie Scene Sequence Object Reference=@n13, DisplayRate_Numerator:int=15, DisplayRate_Denominator:int=1) -> (then)
   > n14 = Add Actor!(
    execute
    self
    Animatable Creature:Animatable Creature Object Reference=Get SelectedAC()
    Level Sequence:Level Sequence Object Reference=@n13
    bIsLayeredControlRig:bool=false
   ) -> (then, New Binding:Movie Scene Binding Proxy Structure) {is=title K2Node_CallFunction}
   > Set CurrentBinding(CurrentBinding:Movie Scene Binding Proxy Structure=@n14) -> (Output_Get:Movie Scene Binding Proxy Structure)
   > AddSpawnableFromClass(
    self=Get LevelSequenceEditorSubsystem()
    Sequence=Get CurrentLvlSeq()
    ClassToSpawn=/AnimInEdTools/AnimationBrowser/Assets/AnimationBrowserStage.AnimationBrowserStage_C
   )
   > n15 = Sequence!(execute) -> (then_0, then_1, then_2:Exec)
   n16 = Get SelectedAC()
   > Is Valid(InputObject=@n16)
   n17 = Get Default Pose!(self:Animatable Creature Object Reference=@n16) -> (Default Pose:Control Rig Pose Asset Soft Object Reference)
   > n18 = Branch(Condition=IsValidSoftObjectReference(SoftObjectReference=@n17))
   n19 = Make Frame Number!(Value:int=1) -> (FrameNumber:Frame Number Structure) {is=title K2Node_MakeStruct}
   > Set Current Local Time!(
    execute
    PlaybackParams:Movie Scene Sequence Playback Params Structure=Make Movie Scene Sequence Playback Params!(
     Frame:Frame Time Structure=Make Frame Time!(FrameNumber:Frame Number Structure=@n19, SubFrame:float=0) -> (FrameTime:Frame Time Structure) {is=title K2Node_MakeStruct}
     Time=0
     MarkedFrame
     Timecode=(Hours=0,Minutes=0,Seconds=0,Frames=0,Subframe=0,bDropFrameFormat=False)
     PositionType=Frame
     UpdateMethod=Play
     bHasJumped=False
    ) -> (MovieSceneSequencePlaybackParams)
    TimeUnit:EMovieSceneTimeUnit Enum=DisplayRate
   ) -> (then) {is=title K2Node_CallFunction}
   n20 = Make Frame Number(Value=0)
   > Set Current Local Time(
    PlaybackParams=Make Movie Scene Sequence Playback Params(
     Frame=Make Frame Time(FrameNumber=@n20, SubFrame=0)
     Time=0
     Timecode=(Hours=0,Minutes=0,Seconds=0,Frames=0,Subframe=0,bDropFrameFormat=False)
     PositionType=Frame
     UpdateMethod=Play
     bHasJumped=False
    )
    TimeUnit=DisplayRate
   )
   > n21 = GetControlRigs!(execute, LevelSequence:Level Sequence Object Reference=Get Level Sequence()) -> (then, ReturnValue:Array of Control Rig Sequencer Binding Proxy Structures)
   > n22 = For Each Loop with Break(Array:Array of Control Rig Sequencer Binding Proxy Structures=@n21, Break=@n24.then) -> (Array Element:Control Rig Sequencer Binding Proxy Structure)
   CheckForAnimSET(execute=@n22.Completed) {is=title K2Node_CallFunction}
   n23 = Get CurrentBinding()
   Branch(
    execute=@n22.LoopBody
    Condition="Equal (Guid)"(
     A:"Guid Structure (by ref)"=Break Movie Scene Binding Proxy!(MovieSceneBindingProxy:"Movie Scene Binding Proxy Structure (by ref)"=@n23) -> (BindingID:Guid Structure, Sequence:Movie Scene Sequence Object Reference) {is=title K2Node_BreakStruct}
     B:"Guid Structure (by ref)"=Break Control Rig Sequencer Binding Proxy!(ControlRigSequencerBindingProxy:"Control Rig Sequencer Binding Proxy Structure (by ref)"=@n22) -> (
      Proxy_BindingID:Guid Structure
      Proxy_Sequence:Movie Scene Sequence Object Reference
      ControlRig:Control Rig Object Reference
      Track:Movie Scene Control Rig Parameter Track Object Reference
     ) {is=title K2Node_BreakStruct}
    ) -> (ReturnValue:bool=~) {is=title K2Node_PromotableOperator}
   )
   > n24 = Paste Pose!(
    execute
    self
    Pose Asset:Soft Object Reference=@n17
    LevelSequence:Level Sequence Object Reference=Get CurrentLvlSeq()
    ControlRig Proxy Binding:"Control Rig Sequencer Binding Proxy Structure (by ref)"=@n22
   ) -> (then)
   CheckForAnimSET(execute=@n18.else)
   n25 = Create Linked Anim Sequence(
    execute=@n15.then_1
    Level Sequence=Get CurrentLvlSeq()
    New AnimSequence Path=Get AnimSeqPath()
    Binding=Get CurrentBinding()
    CustomFrameRange=false
   )
   > n26 = Is Valid(InputObject=Get CurrentSkeleton())
   Set CurrentSkeleton(execute=@n26.Is Not Valid, CurrentSkeleton=GetSkeleton(InAsset=@n25))
   > SaveLoadedAssets(AssetsToSave=Make Array!("[0]"=Get CurrentLvlSeq(), "[1]":Object Reference=@n25) -> (Array), bOnlyIfIsDirty=true)
   n27 = Get ErrorTxt()
   SetVisibility(execute=@n9.else, self=@n27, InVisibility=Visible)
   > Set Text(Text="LOCTEXT(\"Companion BP missing from AC\")", self=@n27)
  ]}
 GetAO {nodes [
   GetAO!() -> (then, AnimName:String) {is=title K2Node_FunctionEntry}
   > n1 = For Each Loop with Break!(Exec, Array:Array of Asset Data Structures=GetAllAnimFolderPaths(), Break=@n10.then) -> (
    LoopBody
    Array Element_PackageName:Name
    Array Element_PackagePath:Name
    Array Element_AssetName:Name
    Array Element_AssetClassPath:Top Level Asset Path Structure
    Array Index
    Completed
   )
   n2 = Is Valid(exec=@n1.Completed, InputObject=Get FoundAO() -> (FoundAO:Aim Offset Blend Space Object Reference))
   n3 = Loop Archetype Leader Assets(execute=@n2.Is Not Valid, Break=@n4.then_0, SubFolder=/Animation/)
   > Branch(Condition=Contains(SearchIn=@n3, Substring=Get AnimName() -> (AnimName:String)))
   > n4 = Sequence!(execute) -> (then_0, then_1)
   n5 = LoadAsset(execute=@n4.then_1, AssetPath=@n3)
   > n6 = Cast To AimOffsetBlendSpace!(execute, Object=@n5) -> (then, CastFailed, AsAim Offset Blend Space:Aim Offset Blend Space Object Reference) {is=title K2Node_DynamicCast}
   > n7 = Set FoundAO(FoundAO:Aim Offset Blend Space Object Reference=@n6) -> (Output_Get:Aim Offset Blend Space Object Reference)
   Return Node!(execute=[@n2.Is Valid, @n7.then], NewParam:Aim Offset Blend Space Object Reference=Get FoundAO())
   Branch(
    execute=@n1.LoopBody
    Condition=Contains(SearchIn="To String (Name)"(InName=@n1.Array Element_AssetName), Substring=Get AnimName(), bUseCase=true)
   )
   > n8 = LoadAsset(AssetPath=Append(A="To String (Name)"(InName=@n1), B=., C="To String (Name)"(InName=@n1.Array Element_AssetName)))
   > n9 = Cast To AimOffsetBlendSpace(Object=@n8)
   > n10 = Set FoundAO(FoundAO=@n9)
  ]}
 GetAllAnimFolderPaths {nodes [
   GetAllAnimFolderPaths() {is=title K2Node_FunctionEntry}
   n1 = Get SelectedAC()
   n2 = Get SelectedAC()
   n3 = Get Variant(self=@n2)
   n4 = Get Asset(self=@n2)
   n5 = GetPath(InPath=Get Path(
     self=Get Naming Convention(self=@n1)
     Token Map=Get Path Parts Map(Asset=@n4, Plugin=@n3)
     Path Pattern=Get Animation Path(self=@n1)
     Remove Spaces On Name=false
    ))
   > n6 = GetAssetsByPaths!(
    execute
    self:Asset Registry Interface=GetAssetRegistry()
    PackagePaths:Array of Names=Make Array!("[0]":Name=String To Name(InString=@n5)) -> (Array:Array of Names)
    bRecursive:bool=true
    bIncludeOnlyOnDiskAssets:bool=false
   ) -> (then, OutAssetData:Array of Asset Data Structures, ReturnValue:bool)
   > Return Node!(execute, OutAssetData:Array of Asset Data Structures=@n6)
  ]}
 GetAllLevelSeqFolderPaths {nodes [
   GetAllLevelSeqFolderPaths() {is=title K2Node_FunctionEntry}
   n1 = Get SelectedAC()
   n2 = Get SelectedAC()
   n3 = Get Variant(self=@n2)
   n4 = Get Asset(self=@n2)
   n5 = GetPath(InPath=Get Path(
     self=Get Naming Convention(self=@n1)
     Token Map=Get Path Parts Map(Asset=@n4, Plugin=@n3)
     Path Pattern=Get Source Path(self=@n1)
     Remove Spaces On Name=false
    ))
   > n6 = GetAssetsByPaths(self=GetAssetRegistry(), PackagePaths=Make Array("[0]"=String To Name(InString=@n5)), bRecursive=true, bIncludeOnlyOnDiskAssets=false)
   > Return Node(OutAssetData=@n6)
  ]}
 GetAnim {nodes [
   n1 = For Each Loop with Break(Array=GetAllAnimFolderPaths() {is=title K2Node_CallFunction})
   n2 = LoadAsset(AssetPath=Append(A="To String (Name)"(InName=@n1), B=., C="To String (Name)"(InName=@n1.Array Element_AssetName)))
   > Is Valid(InputObject=@n2)
   > Cast To AnimSequence(Object=@n2)
   GetAnim!() -> (then, AnimName:String) {is=title K2Node_FunctionEntry}
   n3 = Get Asset(self=Get SelectedAC())
   n4 = Get AnimName()
   > Return Node!(execute, NewParam:Anim Sequence Object Reference=Find(TargetMap=Get AnimCollection(), Key=Append(A=@n3, B=_, C=@n4)))
   "To String (Name)"(InName=@n1.Array Element_AssetName)
  ]}
 GetArchetypeLeaderAssetPath {nodes [
   GetArchetypeLeaderAssetPath() {is=title K2Node_FunctionEntry}
   > n1 = GetDataTableRowNames(Table=/AnimInEdTools/CompanionAnimAssetGenerator/Settings/DT_CompanionArchetype)
   n2 = Get SelectedAC()
   n3 = String To Name(InString=Enum to String(Enumerator=Get Archetype(self=@n2) -> (Archetype=NewEnumerator0)))
   > n4 = Get Data Table Row DT_CompanionArchetype!(
    execute:Exec
    DataTable:Data Table Object Reference=&DT_CompanionArchetype
    RowName:Name="Get (a ref)"(Array=@n1, Dimension 1=Find Item(TargetArray=@n1, ItemToFind=@n3))
   ) -> (
    then:Exec
    RowNotFound:Exec
    ReturnValue_SubFolders_45_#60:Map of Strings to Struct Companion Archetype Anim Names Structures
    ReturnValue_ArchetypeLeaderAssetPath_51_#61:String
   ) {is=title K2Node_GetDataTableRow}
   > Return Node!(execute, LeaderPath:String=@n4.ReturnValue_ArchetypeLeaderAssetPath_51_#61)
  ]}
 GetArchetypeLeaderName {nodes [
   GetArchetypeLeaderName() {is=title K2Node_FunctionEntry}
   n1 = GetArchetypeLeaderAssetPath() {is=title K2Node_CallFunction}
   n2 = FindSubstring(SearchIn=@n1, Substring=/, bSearchFromEnd=true)
   > Return Node!(execute, ReturnValue:String=RightChop(SourceString=@n1, Count=int + int(A=@n2)))
  ]}
 GetBlendSpace1D {nodes [
   GetBlendSpace1D!() -> (then, AnimName:String) {is=title K2Node_FunctionEntry}
   > For Each Loop with Break!(Exec, Array=GetAllAnimFolderPaths(), Break=@n10.then) -> (LoopBody, Array Element:Asset Data Structure, Array Index, Completed)
   > n1 = For Each Loop!(Exec, Array:Array of Asset Data Structures) -> (
    LoopBody
    Array Element_PackageName:Name
    Array Element_PackagePath:Name
    Array Element_AssetName:Name
    Array Element_AssetClassPath_PathString:String
    Array Index
    Completed
   )
   n2 = Is Valid(exec=@n1.Completed, InputObject=Get FoundBS() -> (FoundBS:Blend Space 1D Object Reference))
   n3 = Loop Archetype Leader Assets(execute=@n2.Is Not Valid, Break=@n4.then_0, SubFolder=/Animation/)
   > Branch(Condition=Contains(SearchIn=@n3, Substring=Get AnimName()))
   > n4 = Sequence()
   n5 = LoadAsset(execute=@n4.then_1, AssetPath=@n3)
   > n6 = Cast To BlendSpace1D!(execute, Object=@n5) -> (then, CastFailed, AsBlend Space 1D:Blend Space 1D Object Reference) {is=title K2Node_DynamicCast}
   > n7 = Set FoundBS(FoundBS:Blend Space 1D Object Reference=@n6) -> (Output_Get:Blend Space 1D Object Reference)
   Return Node!(execute=[@n2.Is Valid, @n7.then], NewParam:Blend Space 1D Object Reference=Get FoundBS())
   Branch(
    execute=@n1.LoopBody
    Condition=Contains(SearchIn="To String (Name)"(InName=@n1.Array Element_AssetName), Substring=Get AnimName(), bUseCase=true)
   )
   > n8 = LoadAsset(AssetPath=Append(A="To String (Name)"(InName=@n1), B=., C="To String (Name)"(InName=@n1.Array Element_AssetName)))
   > n9 = Cast To BlendSpace1D(Object=@n8)
   > n10 = Set FoundBS(FoundBS=@n9)
  ]}
 GetBlendSpaceAnims {nodes [
   n1 = GetBlendSpaceAnims!() -> (then, MapPin:Map of Strings to Vectors) {is=title K2Node_FunctionEntry}
   > n2 = "For Each Loop (Map)"(MapPin:Map of Strings to Vectors=@n1) -> (KeyPin:String, ValuePin:Vector)
   > n3 = For Each Loop(Array=GetAllAnimFolderPaths())
   > Branch(Condition=Contains(SearchIn="To String (Name)"(InName=@n3.Array Element_AssetName), Substring=@n2, bUseCase=true))
   > n4 = LoadAsset(AssetPath=Append(A="To String (Name)"(InName=@n3), B=., C="To String (Name)"(InName=@n3.Array Element_AssetName)))
   > n5 = Cast To AnimSequence(Object=@n4)
   > Add(
    TargetArray:Array of Anim Sequence Object References=Get Anims() -> (Anims:Array of Anim Sequence Object References)
    NewItem:"Anim Sequence Object Reference (by ref)"=@n5
   )
   > Set found(found=true)
   Branch(execute=@n3.Completed, Condition=Get found())
   n6 = Get Asset(self=Get SelectedAC())
   > RaiseScriptError!(execute, ErrorMessage:String=Append!(A="ANIMATION MISSING: ", B=Append!(A=@n6, B=_, C:String=@n2) -> (ReturnValue)) -> (ReturnValue)) -> (then)
   > Set found(found=false)
   Return Node!(execute=@n2.CompletedPin, BlendSpaceAnims:Array of Anim Sequence Object References=Get Anims())
  ]}
 GetChooserPRT {nodes [
   GetChooserPRT() {is=title K2Node_FunctionEntry}
   n1 = GetArchetypeLeaderName() {is=title K2Node_CallFunction}
   n2 = Append!(A=GetArchetypeLeaderAssetPath(), B=&Choosers/) -> (ReturnValue)
   n3 = Get Asset(self=Get SelectedAC())
   n4 = Append!(A=PRT, B=_, C:String=Get Asset(self=Get SelectedAC()), D:String=_Choosers) -> (ReturnValue)
   > n5 = LoadAsset(AssetPath=Append(A=Replace(SourceString=@n2, From=@n1, To=@n3), B=@n4, C=., D=@n4))
   > n6 = Cast To ProxyTable!(execute, Object=@n5) -> (then, CastFailed, AsProxy Table:Proxy Table Object Reference) {is=title K2Node_DynamicCast}
   > Return Node!(execute, ReturnValue:Proxy Table Object Reference=@n6)
  ]}
 GetDataAsset {nodes [
   GetDataAsset() {is=title K2Node_FunctionEntry}
   n1 = GetArchetypeLeaderName()
   n2 = Append!(A=GetArchetypeLeaderAssetPath(), B=/Blueprint/) -> (ReturnValue)
   n3 = Get Asset(self=Get SelectedAC())
   n4 = Append(A=Get Asset(self=Get SelectedAC()), B=_DA)
   > n5 = LoadAsset(AssetPath=Append!(A=Replace(SourceString=@n2, From=@n1, To=@n3), B=@n4, C:String=., D:String=@n4) -> (ReturnValue))
   > n6 = Cast To FortUserDefinedDataAsset(Object=@n5)
   > Return Node(ReturnValue:Fort User Defined Data Asset Object Reference=@n6)
  ]}
 GetDataAssetAnim {nodes [
   n1 = GetDataAssetAnim!() -> (then, AnimName:String) {is=title K2Node_FunctionEntry}
   n2 = GetAnim(AnimName=@n1) {is=title K2Node_CallFunction}
   > n3 = Is Valid(InputObject=@n2)
   n4 = Loop Archetype Leader Assets(execute=@n3.Is Not Valid, Break=@n5.then_0, SubFolder=/Animation/)
   > Branch(Condition=Contains(SearchIn=@n4, Substring=Get AnimName()))
   > n5 = Sequence()
   n6 = LoadAsset(execute=@n5.then_1, AssetPath=@n4)
   > n7 = Cast To AnimSequence(Object=@n6)
   > n8 = Set AnimSeq(AnimSeq:Anim Sequence Object Reference=@n7) -> (Output_Get:Anim Sequence Object Reference)
   Return Node!(execute=[@n8.then, @n9.then], NewParam:Anim Sequence Object Reference=Get AnimSeq() -> (AnimSeq:Anim Sequence Object Reference))
   n9 = Set AnimSeq(execute=@n3.Is Valid, AnimSeq=@n2)
  ]}
 GetItems {nodes [
   GetItems() {is=title K2Node_FunctionEntry}
   > n1 = Set CompanionACs(CompanionACs:Array of Animatable Creature Object References=Get Animatable Creatures!(
     Project:BP Animation Browser Project Object Reference=Get CompanionsProjectRef()
     bSearchSubClasses:bool=true
     __WorldContext:Object Reference
    ) -> (Animatable Creatures:Array of Animatable Creature Object References) {is=title K2Node_CallFunction}) -> (Output_Get:Array of Animatable Creature Object References)
   > n2 = For Each Loop!(Exec, Array:Array of Animatable Creature Object References=@n1) -> (LoopBody, Array Element:Animatable Creature Object Reference, Array Index, Completed)
   n3 = Execute Python Script!(
    execute=@n2.Completed
    PythonScript="sorted_creatures = sorted(creatures)\r\n\r\nif 'Shared' in sorted_creatures:\r\n    sorted_creatures.insert(0, sorted_creatures.pop(sorted_creatures.index('Shared')))\r\n\r\nif len(creatures) > 1 and add_all:\r\n    sorted_creatures.insert(0, 'All')"
    creatures:Array of Strings=Get AC_Items() -> (AC_Items:Array of Strings)
    add_all:bool=Get Add All Option() -> (Add All Option:bool)
   ) -> (then, ReturnValue, sorted_creatures:Array of Strings) {inputs [creatures,add_all], outputs [sorted_creatures]}
   > Return Node!(execute, Items:Array of Strings=@n3.sorted_creatures)
   Add(execute=@n2.LoopBody, TargetArray:Array of Strings=Get AC_Items(), NewItem:"String (by ref)"=Get Asset(self=@n2))
  ]}
 GetSelectedAC {nodes [
   n1 = GetSelectedAC!() -> (then, InputPin:String) {is=title K2Node_FunctionEntry}
   > n2 = For Each Loop(Array=Get CompanionACs() -> (CompanionACs:Array of Animatable Creature Object References))
   > Branch(Condition="Equal, Case Insensitive (String)"(A=@n1, B=Get Asset(self=@n2)))
   > Set SelectedAC(SelectedAC:Animatable Creature Object Reference=@n2) -> (Output_Get:Animatable Creature Object Reference)
  ]}
 HandleCheckboxes {nodes [
   n1 = HandleCheckboxes!() -> (then, InputPin:bool=~) {is=title K2Node_FunctionEntry}
   > n2 = Set AllHolders(AllHolders:Array of Widget Object References=GetAllChildren(self=Get SequenceColumnHolder())) -> (Output_Get:Array of Widget Object References)
   > Append Array!(
    execute
    TargetArray:Array of Widget Object References=@n2
    SourceArray:Array of Widget Object References=GetAllChildren(self=Get BlendSpaceColumnHolder())
   ) -> (then)
   > Append Array(
    TargetArray=Get AllHolders() -> (AllHolders:Array of Widget Object References)
    SourceArray=GetAllChildren(self=Get ChooserTableColumnHolder())
   )
   > n3 = For Each Loop(Array:Array of Widget Object References=Get AllHolders()) -> (Array Element:Widget Object Reference)
   > n4 = Cast To AnimLists(Object=@n3)
   > n5 = Branch(Condition=@n1)
   n6 = Get AllColumnItemsCheckBox!(self:Anim Lists Object Reference=@n4) -> (AllColumnItemsCheckBox:Editor Utility Check Box Object Reference)
   n7 = Get AnimBPCheckBox()
   n8 = Get DataAssetCheckBox()
   n9 = Get MontageCheckBox()
   n10 = Get CASCheckBox()
   > n11 = SetCheckedState(self=[@n6, @n7, @n8, @n9, @n10], InCheckedState=Checked)
   CheckPassThrough!(execute=[@n11.then, @n12.then], self:Anim Lists Object Reference=@n4, Checked:bool=IsChecked(self=@n6)) -> (then)
   n12 = SetCheckedState(execute=@n5.else, self=[@n6, @n7, @n8, @n9, @n10], InCheckedState=Unchecked)
  ]}
 LoopAllAnims {nodes [
   Inputs!() -> (Exec:Exec)
   > n1 = For Each Loop!(Exec, Array:Array of Asset Data Structures=GetAllAnimFolderPaths()) -> (
    LoopBody
    Array Element_PackageName:Name
    Array Element_PackagePath:Name
    Array Element_AssetName:Name
    Array Element_AssetClassPath_PathString:String
    Array Index
    Completed
   )
   > n2 = LoadAsset(AssetPath=Append!(A="To String (Name)"(InName=@n1), B=., C="To String (Name)"(InName=@n1.Array Element_AssetName)) -> (ReturnValue))
   > n3 = Cast To AnimSequence(Object=@n2)
   > Outputs!(then:Exec, Completed:Exec=@n1.Completed, AsAnim Sequence:Anim Sequence Object Reference=@n3)
  ]}
 LoopArchetypeLeaderAssets {nodes [
   n1 = Inputs!() -> (execute:Exec, Break:Exec, SubFolder:String)
   n2 = Append!(A=GetArchetypeLeaderAssetPath(), B=@n1) -> (ReturnValue)
   > n3 = GetAssetsByPaths(self=GetAssetRegistry(), PackagePaths=Make Array("[0]"=String To Name(InString=@n2)), bRecursive=true, bIncludeOnlyOnDiskAssets=false)
   > n4 = For Each Loop with Break!(Exec, Array=@n3, Break=@n1.Break) -> (
    LoopBody
    Array Element_PackageName:Name
    Array Element_PackagePath:Name
    Array Element_AssetName:Name
    Array Element_AssetClassPath:Top Level Asset Path Structure
    Array Index
    Completed
   )
   > Outputs!(
    LoopBody:Exec
    AssetPathAndName:String=Append!(A="To String (Name)"(InName=@n4), B=., C:String="To String (Name)"(InName=@n4.Array Element_AssetName)) -> (ReturnValue)
   )
  ]}
 MiscAnimSettings {nodes [
   MiscAnimSettings() {is=title K2Node_FunctionEntry}
   > n1 = "For Each Loop (Map)"(MapPin:Map of Strings to Anim Sequence Object References=Get AnimCollection()) -> (ValuePin:Anim Sequence Object Reference)
   > n2 = Set currentAnim(currentAnim:Anim Sequence Object Reference=@n1.ValuePin) -> (Output_Get:Anim Sequence Object Reference)
   n3 = GetDisplayName(Object=@n1.ValuePin)
   > n4 = Branch(Condition=OR Boolean!(
     A=Contains(SearchIn=@n3, Substring=Worried_Loco)
     B=Contains(SearchIn=@n3, Substring=Angry_Loco)
     C:bool=Contains(SearchIn=@n3, Substring=Happy_Loco)
     D:bool=Contains(SearchIn=@n3, Substring=_Lp)
    ) -> (ReturnValue))
   n5 = Get currentAnim() -> (currentAnim:Anim Sequence Object Reference)
   > Execute Python Script!(
    execute
    PythonScript="import unreal\r\n\r\nanim = anim\r\nanim.set_editor_property(\"bLoop\", True)\r\n"
    anim:Anim Sequence Object Reference=@n5
   ) -> (then, ReturnValue) {inputs [anim]}
   Branch(execute=@n4.else, Condition=Contains(SearchIn=GetDisplayName(Object=@n2), Substring=Interact_))
   > SetRootMotionEnabled!(execute, AnimationSequence:Anim Sequence Object Reference=@n5, bEnabled:bool=true) -> (then)
  ]}
 Paste Pose {nodes [
   n1 = Paste Pose!() -> (
    then
    Pose Asset:Soft Object Reference
    LevelSequence:Level Sequence Object Reference
    ControlRig Proxy Binding:"Control Rig Sequencer Binding Proxy Structure (by ref)"
   ) {is=title K2Node_FunctionEntry}
   > n2 = LoadAsset_Blocking(Asset=@n1)
   > Is Valid(InputObject=@n2)
   > n3 = Cast To ControlRigPoseAsset!(execute, Object=@n2) -> (then, CastFailed, AsControl Rig Pose Asset:Control Rig Pose Asset Object Reference) {is=title K2Node_DynamicCast}
   > n4 = GetControlRigs(LevelSequence=Get LevelSequence() -> (LevelSequence:Level Sequence Object Reference))
   > Branch(Condition=Is Not Empty!(TargetArray:Array of Control Rig Sequencer Binding Proxy Structures=@n4) -> (ReturnValue:bool))
   n5 = Get ControlRig Proxy Binding() -> (ControlRig Proxy Binding:"Control Rig Sequencer Binding Proxy Structure (by ref)")
   > n6 = Execute Python Script!(
    execute
    PythonScript="import unreal\r\n\r\nnum_keys = 0\r\nfor section in proxy.track.get_sections():\r\n    float_channels = section.get_channels_by_type(unreal.MovieSceneScriptingFloatChannel)\r\n    for channel in float_channels:\r\n        num_keys += channel.get_num_keys()\r\n\r\nresult = num_keys == 0\r\n"
    proxy:"Control Rig Sequencer Binding Proxy Structure (by ref)"=@n5
   ) -> (then, ReturnValue, result:bool=~) {inputs [proxy], outputs [result]}
   > Branch(Condition=@n6.result)
   n7 = Break Control Rig Sequencer Binding Proxy!(ControlRigSequencerBindingProxy=@n5) -> (Proxy:Movie Scene Binding Proxy Structure, ControlRig, Track)
   > SelectControls!(
    execute
    self:Control Rig Pose Asset Object Reference=@n3
    InControlRig:Control Rig Object Reference=@n7.ControlRig
    bDoMirror:bool=false
    bClearSelection:bool=true
   ) -> (then)
   > PastePose!(
    execute
    self:Control Rig Pose Asset Object Reference=@n3
    InControlRig:Control Rig Object Reference=@n7.ControlRig
    bDoKey:bool=true
    bDoMirror:bool=false
    bDoAdditive:bool=false
   ) -> (then)
   > ClearControlSelection!(execute, self:Control Rig Object Reference=@n7.ControlRig, bSetupUndo:bool=false) -> (then, ReturnValue:bool)
   "Get (a copy)"(Array:Array of Control Rig Sequencer Binding Proxy Structures=@n4) -> (Output:Control Rig Sequencer Binding Proxy Structure)
  ]}
 RenameDuplicate {nodes [
   n1 = RenameDuplicate!() -> (then, Object:Object Reference) {is=title K2Node_FunctionEntry}
   n2 = GetArchetypeLeaderAssetPath()
   n3 = int + int(A=FindSubstring(SearchIn=@n2, Substring=/, bSearchFromEnd=true))
   n4 = Get SelectedAC()
   > Return Node!(
    execute
    ReturnValue:String=Replace(SourceString=GetDisplayName(Object=@n1), From=RightChop(SourceString=@n2, Count=@n3), To=Get Asset(self=@n4))
   )
  ]}
 SaveAllAssets {nodes [
   SaveAllAssets() {is=title K2Node_FunctionEntry}
   > n1 = Values(TargetMap:Map of Strings to Anim Sequence Object References=Get AnimCollection()) -> (Values:Array of Anim Sequence Object References)
   > SaveLoadedAssets(AssetsToSave=@n1, bOnlyIfIsDirty=true)
   > n2 = For Each Loop(Array=GetAllLevelSeqFolderPaths!() -> (OutAssetData:Array of Asset Data Structures) {is=title K2Node_CallFunction})
   > n3 = LoadAsset(AssetPath=Append(A="To String (Name)"(InName=@n2), B=., C="To String (Name)"(InName=@n2.Array Element_AssetName)))
   > SaveLoadedAssets(AssetsToSave=Make Array("[0]":Object Reference=@n3) -> (Array:Array of Object References), bOnlyIfIsDirty=true)
   n4 = For Each Loop(Array=GetAllAnimFolderPaths())
   > n5 = LoadAsset(AssetPath=Append(A="To String (Name)"(InName=@n4), B=., C="To String (Name)"(InName=@n4.Array Element_AssetName)))
   > SaveLoadedAssets(AssetsToSave=Make Array("[0]"=@n5), bOnlyIfIsDirty=true)
  ]}
 SetAdditiveSettings {nodes [
   SetAdditiveSettings() {is=title K2Node_FunctionEntry}
   > n1 = Sequence()
   > n2 = Loop All Anims!(Exec:Exec) -> (then:Exec, Completed:Exec, AsAnim Sequence:Anim Sequence Object Reference) {is=title K2Node_MacroInstance}
   > Add(
    TargetMap:Map of Strings to Anim Sequence Object References=Get AnimCollection()
    Key=GetDisplayName(Object=@n2)
    Value:"Anim Sequence Object Reference (by ref)"=@n2
   ) {is=title K2Node_CallFunction}
   n3 = "For Each Loop (Map)"(execute=@n1.then_1, MapPin=Get AnimCollection())
   n4 = CheckNeedsAdditive!(AnimSequence:Anim Sequence Object Reference=@n3.ValuePin) -> (
    HasAdditive:bool
    AnimSeq:Anim Sequence Object Reference
    AdditiveSettings_Anim_3_#62:String
    AdditiveSettings_AdditiveType_7_#63:ENUM_AnimAdditiveType Enum
    AdditiveSettings_PoseType_10_#64:ENUM_AnimAdditiveBasePoseType Enum
   ) {is=title K2Node_CallFunction}
   > Branch(Condition=@n4)
   > Set Animation(Animation:Anim Sequence Object Reference=@n4.AnimSeq) -> (Output_Get:Anim Sequence Object Reference)
   > Set AdditiveToUse(AdditiveToUse:String=@n4.AdditiveSettings_Anim_3_#62) -> (Output_Get:String)
   > Set AdditiveType(AdditiveType:ENUM_AnimAdditiveType Enum=@n4.AdditiveSettings_AdditiveType_7_#63) -> (Output_Get:ENUM_AnimAdditiveType Enum=NewEnumerator0)
   > Set BasePose(BasePose:ENUM_AnimAdditiveBasePoseType Enum=@n4.AdditiveSettings_PoseType_10_#64) -> (Output_Get:ENUM_AnimAdditiveBasePoseType Enum=NewEnumerator0)
   n5 = Get AdditiveType() -> (AdditiveType:ENUM_AnimAdditiveType Enum=NewEnumerator0)
   > SetAdditiveAnimationType!(
    execute
    AnimationSequence:Anim Sequence Object Reference=Get Animation() -> (Animation:Anim Sequence Object Reference)
    AdditiveAnimationType:EAdditiveAnimationType Enum=Byte to Enum EAdditiveAnimationType!(Byte:Byte=MakeLiteralByte!(Value:Byte=@n5) -> (ReturnValue:Byte=0)) -> (ReturnValue:EAdditiveAnimationType Enum) {is=title K2Node_CastByteToEnum}
   ) -> (then)
   n6 = Get BasePose() -> (BasePose:ENUM_AnimAdditiveBasePoseType Enum=NewEnumerator0)
   > SetAdditiveBasePoseType!(
    execute
    AnimationSequence:Anim Sequence Object Reference=Get Animation()
    AdditiveBasePoseType:EAdditiveBasePoseType Enum=Byte to Enum EAdditiveBasePoseType!(Byte:Byte=MakeLiteralByte(Value=@n6) -> (ReturnValue=0)) -> (ReturnValue:EAdditiveBasePoseType Enum) {is=title K2Node_CastByteToEnum}
   ) -> (then)
   n7 = GetAnim(AnimName=Get AdditiveToUse() -> (AdditiveToUse:String))
   > Is Valid(InputObject=@n7)
   > Execute Python Script!(
    execute
    PythonScript="import unreal\r\n\r\n## Load your animation sequence\r\nanim = anim\r\n\r\n## Load the base pose sequence you want to use\r\nbase_pose = pose_anim\r\n\r\n\r\n\r\n## Assign the actual base animation sequence\r\nanim.set_editor_property(\"RefPoseSeq\", base_pose)\r\n\r\n"
    anim:Anim Sequence Object Reference=Get Animation()
    pose_anim:Anim Sequence Object Reference=@n7
   ) -> (then, ReturnValue) {inputs [anim,pose_anim]}
  ]}
 SetColumns {nodes [
   SetColumns() {is=title K2Node_FunctionEntry}
   > ClearChildren(self=[Get ChooserTableColumnHolder(), Get BlendSpaceColumnHolder(), Get SequenceColumnHolder()])
   > n1 = GetDataTableRowNames(Table=&DT_CompanionArchetype)
   n2 = Get SelectedAC()
   n3 = String To Name(InString=Enum to String(Enumerator=Get Archetype(self=@n2) -> (Archetype=NewEnumerator0)))
   > n4 = Get Data Table Row DT_CompanionArchetype(
    DataTable=&DT_CompanionArchetype
    RowName="Get (a ref)"(Array=@n1, Dimension 1=Find Item(TargetArray=@n1, ItemToFind=@n3))
   )
   > n5 = Keys!(execute, TargetMap:Map of Strings to Struct Companion Archetype Anim Names Structures=@n4) -> (then, Keys:Array of Strings)
   > n6 = For Each Loop!(Exec, Array:Array of Strings=@n5) -> (LoopBody, Array Element:String, Array Index, Completed)
   > n7 = Create Widget!(
    execute:Exec
    WorldContextObject:Object Reference
    Class:User Widget Class Reference=/AnimInEdTools/CompanionAnimAssetGenerator/SubWidgets/AnimLists.AnimLists_C
    OwningPlayer:Player Controller Object Reference
   ) -> (then:Exec, ReturnValue:Anim Lists Object Reference)
   > n8 = Branch(Condition=Contains(SearchIn=@n6, Substring=Choosers))
   > n9 = AddChildToStackBox!(execute, self:Stack Box Object Reference=Get ChooserTableColumnHolder(), Content:Widget Object Reference=@n7) -> (then, ReturnValue:Stack Box Slot Object Reference)
   n10 = Set ColumnRef(execute=[@n12.then, @n9.then, @n13.then], ColumnRef:Anim Lists Object Reference=@n7) -> (Output_Get:Anim Lists Object Reference)
   > SetupList!(
    execute
    self:Anim Lists Object Reference=@n10
    ColumnName:String=@n6
    AC Ref:Animatable Creature Object Reference=Get SelectedAC()
    SubFolderLocation:int=@n6.Array Index
   ) -> (then)
   n11 = Branch(execute=@n8.else, Condition=Contains(SearchIn=@n6, Substring=Blend Spaces))
   > n12 = AddChildToStackBox(self=Get BlendSpaceColumnHolder(), Content=@n7)
   n13 = AddChildToStackBox(execute=@n11.else, self=Get SequenceColumnHolder(), Content=@n7)
  ]}
 UpdateComboBox {nodes [
   n1 = UpdateComboBox!() -> (then, ComboBox:"Searchable Combo Box (String) Object Reference", DefaultItem:String, Items:Array of Strings, AddBlank:bool=~) {is=title K2Node_FunctionEntry}
   > n2 = Set Current Index(Current Index:int=GetSelectedIndex!(self:"Searchable Combo Box (String) Object Reference"=@n1) -> (ReturnValue:int=0)) -> (Output_Get:int=0)
   > Set Selected Item(Selected Item:String=GetOptionAtIndex!(
     self:"Searchable Combo Box (String) Object Reference"=Get ComboBox() -> (ComboBox:"Searchable Combo Box (String) Object Reference")
     Index:int=@n2
    ) -> (ReturnValue:String)) -> (Output_Get:String)
   n3 = Get ComboBox()
   > n4 = For Loop(LastIndex=int - int(A=GetOptionCount!(self:"Searchable Combo Box (String) Object Reference"=@n3) -> (ReturnValue:int=0)))
   n5 = Branch(
    execute=@n4.Completed
    Condition=Identical!(ArrayA:Array of Strings=Get Items() -> (Items:Array of Strings), ArrayB:Array of Strings=Get Old Items() -> (Old Items:Array of Strings)) -> (ReturnValue:bool) {is=title K2Node_CallArrayFunction}
   )
   ClearOptions(execute=@n5.else, self:"Searchable Combo Box (String) Object Reference"=Get ComboBox())
   > n6 = Sequence!(execute) -> (then_0, then_1, then_2:Exec, then_3:Exec)
   > Branch(Condition=Get AddBlank() -> (AddBlank:bool))
   > AddOption!(execute, self:"Searchable Combo Box (String) Object Reference"=Get ComboBox(), Option:String) -> (then)
   n7 = Branch(execute=@n6.then_1, Condition=IsEmpty(InString=Get DefaultItem() -> (DefaultItem:String)))
   AddOption(execute=@n7.else, self=Get ComboBox())
   n8 = For Each Loop(Exec=@n6.then_2, Array=Get Items())
   > AddOption(self=Get ComboBox(), Option=@n8)
   n9 = Get ComboBox()
   n10 = Get Current Index() -> (Current Index:int=0)
   n11 = Branch(
    execute=@n6.then_3
    Condition="Equal Exactly (String)"(A=GetOptionAtIndex(self=@n9, Index=@n10), B=Get Selected Item() -> (Selected Item:String))
   )
   n12 = Get ComboBox()
   > AddOption(self=@n12, Option=Temp)
   n13 = Get Current Index() -> (Current Index=0)
   > SetSelectedIndex(self=@n12, Index=int + int(A=@n13))
   > SetSelectedIndex(self=@n12, Index=@n13)
   > RemoveOption!(execute, self:"Searchable Combo Box (String) Object Reference"=@n12, Option:String=Temp) -> (then, ReturnValue:bool)
   ClearSelection(execute=@n11.else, self:"Searchable Combo Box (String) Object Reference"=Get ComboBox())
   n14 = GetOptionAtIndex(self=Get ComboBox(), Index=@n4)
   n15 = Branch(execute=@n4.LoopBody, Condition="Equal Exactly (String)"(A=@n14, B=Get DefaultItem()))
   Add(execute=@n15.else, TargetArray=Get Old Items(), NewItem=@n14) {is=title K2Node_CallArrayFunction}
  ]}
]
variables [
 AnimCollection {type="map<AnimSequence>", default {}}
 AnimSeqPath {type=string, default=""}
 CompanionACs {type="array<AnimatableCreature_C>", default []}
 CompanionsProjectRef {type=BP_AnimationBrowser_Project_C, default=null}
 CurrentAnimName {type=string, default=""}
 CurrentBinding {type=MovieSceneBindingProxy, default {bindingId=#65, sequence=null}}
 CurrentLvlSeq {type=LevelSequence, default=null}
 CurrentSkeleton {type=Skeleton, default=null}
 CurrentSubfolder {type=string, default=""}
 LvlSeqPath {type=string, default=""}
 SelectedAC {type=AnimatableCreature_C, default=null}
]
widgets {
 info {rootWidgetClass=&CanvasPanel, widgetCount=34}
 widgets [
  CanvasPanel {widgetClassPath=&CanvasPanel, bIsVariable=true, widget=CanvasPanel}
  VerticalBox_0 {
   widgetClassPath=&VerticalBox
   parent=CanvasPanel
   widget=VerticalBox_0
   slot {class=CanvasPanelSlot, properties {layoutData {alignment=<0 0>, anchors {maximum=<1 1>, minimum=<0 0>}, offsets=<ltrb 0 0 0 0>}}}
  }
  HeaderBorder {
   widgetClassPath=&Border
   parent=VerticalBox_0
   widget=HeaderBorder
   properties {brushColor=<rgba 0.0275 0.0275 0.0275 1>, padding=<ltrb 2 2 2 2>}
   slot {class=VerticalBoxSlot, properties {padding=<ltrb 0 0 0 1>}}
  }
  VerticalBox_1 {
   widgetClassPath=&VerticalBox
   parent=HeaderBorder
   widget=VerticalBox_1
   slot {class=BorderSlot, properties {padding=<ltrb 2 2 2 2>}}
  }
  HorizontalBox_1 {widgetClassPath=&HorizontalBox, parent=VerticalBox_1, widget=HorizontalBox_1, slot {class=VerticalBoxSlot}}
  AssetImage {
   widgetClassPath=&Image
   bIsVariable=true
   parent=HorizontalBox_1
   widget=AssetImage
   properties {brush {
     bIsDynamicallyLoaded=false
     drawAs=Image
     imageType=FullColor
     mirroring=NoMirror
     resourceName=None
     resourceObject=/AnimInEdTools/AnimationBrowser/Icons/Icon_Asset
     tiling=NoTile
     imageSize=<18 28>
     margin=<ltrb 0 0 0 0>
     outlineSettings {
      bUseBrushTransparency=false
      roundingType=HalfHeightRadius
      width=0
      color {colorUseRule=UseColor_Specified, specifiedColor=<rgba 0 0 0 0>}
      cornerRadii=<0 0 0 0>
     }
     tintColor {colorUseRule=UseColor_Specified, specifiedColor=<rgba 1 1 1 1>}
     uVRegion {bIsValid=false, max=<0 0>, min=<0 0>}
    }}
   slot {class=HorizontalBoxSlot, properties {horizontalAlignment=HAlign_Center, verticalAlignment=VAlign_Center}}
  }
  SizeBox_3 {
   widgetClassPath=&SizeBox
   parent=HorizontalBox_1
   widget=SizeBox_3
   properties {
    bOverride_MaxDesiredHeight=true
    bOverride_MaxDesiredWidth=true
    bOverride_MinDesiredWidth=true
    maxDesiredHeight=25
    maxDesiredWidth=200
    minDesiredWidth=100
   }
   slot {class=HorizontalBoxSlot, properties {size {sizeRule=Fill, value=1}}}
  }
  Asset_ComboBox {
   widgetClassPath=/Script/FortniteEditor.SearchableStringComboBox
   bIsVariable=true
   parent=SizeBox_3
   widget=Asset_ComboBox
   properties {
    selectedOption=Category
    defaultOptions [Category]
    font {
     bForceMonospaced=false
     bMaterialIsStencil=false
     fontMaterial=null
     fontObject=&Roboto
     letterSpacing=0
     monospacedWidth=1
     size=12
     skewAmount=0
     typefaceFontName=Bold
     outlineSettings {
      bApplyOutlineToDropShadows=false
      bMiteredCorners=false
      bSeparateFillAlpha=false
      outlineMaterial=null
      outlineSize=0
      outlineColor=<rgba 0 0 0 1>
     }
    }
    foregroundColor {colorUseRule=UseColor_Specified, specifiedColor=<rgba 0.5271 0.5271 0.5271 1>}
    itemStyle {
     bUseParentRowBrush=false
     activeBrush {
      bIsDynamicallyLoaded=false
      drawAs=Image
      imageType=NoImage
      mirroring=NoMirror
      resourceName=None
      resourceObject=null
      tiling=NoTile
      imageSize=<0 0>
      margin=<ltrb 0 0 0 0>
      outlineSettings {
       bUseBrushTransparency=false
       roundingType=HalfHeightRadius
       width=0
       color {colorUseRule=UseColor_Specified, specifiedColor=<rgba 0 0 0 0>}
       cornerRadii=<0 0 0 0>
      }
      tintColor {colorUseRule=UseColor_Specified, specifiedColor=<rgba 0 0.162 0.7454 1>}
      uVRegion {bIsValid=false, max=<0 0>, min=<0 0>}
     }
     activeHighlightedBrush {
      bIsDynamicallyLoaded=false
      drawAs=Image
      imageType=NoImage
      mirroring=NoMirror
      resourceName=None
      resourceObject=null
      tiling=NoTile
      imageSize=<0 0>
      margin=<ltrb 0 0 0 0>
      outlineSettings {
       bUseBrushTransparency=false
       roundingType=HalfHeightRadius
       width=0
       color {colorUseRule=UseColor_Specified, specifiedColor=<rgba 0 0 0 0>}
       cornerRadii=<0 0 0 0>
      }
      tintColor {colorUseRule=UseColor_Specified, specifiedColor=<rgba 0.0252 0.0319 0.0423 1>}
      uVRegion {bIsValid=false, max=<0 0>, min=<0 0>}
     }
     activeHoveredBrush {
      bIsDynamicallyLoaded=false
      drawAs=Image
      imageType=NoImage
      mirroring=NoMirror
      resourceName=None
      resourceObject=null
      tiling=NoTile
      imageSize=<0 0>
      margin=<ltrb 0 0 0 0>
      outlineSettings {
       bUseBrushTransparency=false
       roundingType=HalfHeightRadius
       width=0
       color {colorUseRule=UseColor_Specified, specifiedColor=<rgba 0 0 0 0>}
       cornerRadii=<0 0 0 0>
      }
      tintColor {colorUseRule=UseColor_Specified, specifiedColor=<rgba 0 0.162 0.7454 1>}
      uVRegion {bIsValid=false, max=<0 0>, min=<0 0>}
     }
     dropIndicator_Above {
      bIsDynamicallyLoaded=false
      drawAs=Box
      imageType=FullColor
      mirroring=NoMirror
      resourceName=../../../Engine/Content/Slate/Common/DropZoneIndicator_Above.png
      resourceObject=null
      tiling=NoTile
      imageSize=<0 0>
      margin=<ltrb 0.625 0.625 0 0>
      outlineSettings {
       bUseBrushTransparency=false
       roundingType=HalfHeightRadius
       width=0
       color {colorUseRule=UseColor_Specified, specifiedColor=<rgba 0 0 0 0>}
       cornerRadii=<0 0 0 0>
      }
      tintColor {colorUseRule=UseColor_Specified, specifiedColor=<rgba 0.0091 0.3515 0.9647 1>}
      uVRegion {bIsValid=false, max=<0 0>, min=<0 0>}
     }
     dropIndicator_Below {
      bIsDynamicallyLoaded=false
      drawAs=Box
      imageType=FullColor
      mirroring=NoMirror
      resourceName=../../../Engine/Content/Slate/Common/DropZoneIndicator_Below.png
      resourceObject=null
      tiling=NoTile
      imageSize=<0 0>
      margin=<ltrb 0.625 0 0 0.625>
      outlineSettings {
       bUseBrushTransparency=false
       roundingType=HalfHeightRadius
       width=0
       color {colorUseRule=UseColor_Specified, specifiedColor=<rgba 0 0 0 0>}
       cornerRadii=<0 0 0 0>
      }
      tintColor {colorUseRule=UseColor_Specified, specifiedColor=<rgba 0.0091 0.3515 0.9647 1>}
      uVRegion {bIsValid=false, max=<0 0>, min=<0 0>}
     }
     dropIndicator_Onto {
      bIsDynamicallyLoaded=false
      drawAs=Box
      imageType=FullColor
      mirroring=NoMirror
      resourceName=../../../Engine/Content/Slate/Common/DropZoneIndicator_Onto.png
      resourceObject=null
      tiling=NoTile
      imageSize=<0 0>
      margin=<ltrb 0.25 0.25 0.25 0.25>
      outlineSettings {
       bUseBrushTransparency=false
       roundingType=HalfHeightRadius
       width=0
       color {colorUseRule=UseColor_Specified, specifiedColor=<rgba 0 0 0 0>}
       cornerRadii=<0 0 0 0>
      }
      tintColor {colorUseRule=UseColor_Specified, specifiedColor=<rgba 0.0091 0.3515 0.9647 1>}
      uVRegion {bIsValid=false, max=<0 0>, min=<0 0>}
     }
     evenRowBackgroundBrush {
      bIsDynamicallyLoaded=false
      drawAs=Image
      imageType=NoImage
      mirroring=NoMirror
      resourceName=None
      resourceObject=null
      tiling=NoTile
      imageSize=<0 0>
      margin=<ltrb 0 0 0 0>
      outlineSettings {
       bUseBrushTransparency=false
       roundingType=HalfHeightRadius
       width=0
       color {colorUseRule=UseColor_Specified, specifiedColor=<rgba 0 0 0 0>}
       cornerRadii=<0 0 0 0>
      }
      tintColor {colorUseRule=UseColor_Specified, specifiedColor=<rgba 0.0103 0.0103 0.0103 1>}
      uVRegion {bIsValid=false, max=<0 0>, min=<0 0>}
     }
     evenRowBackgroundHoveredBrush {
      bIsDynamicallyLoaded=false
      drawAs=Image
      imageType=NoImage
      mirroring=NoMirror
      resourceName=None
      resourceObject=null
      tiling=NoTile
      imageSize=<0 0>
      margin=<ltrb 0 0 0 0>
      outlineSettings {
       bUseBrushTransparency=false
       roundingType=HalfHeightRadius
       width=0
       color {colorUseRule=UseColor_Specified, specifiedColor=<rgba 0 0 0 0>}
       cornerRadii=<0 0 0 0>
      }
      tintColor {colorUseRule=UseColor_Specified, specifiedColor=<rgba 0.0176 0.0176 0.0176 1>}
      uVRegion {bIsValid=false, max=<0 0>, min=<0 0>}
     }
     inactiveBrush {
      bIsDynamicallyLoaded=false
      drawAs=Image
      imageType=NoImage
      mirroring=NoMirror
      resourceName=None
      resourceObject=null
      tiling=NoTile
      imageSize=<0 0>
      margin=<ltrb 0 0 0 0>
      outlineSettings {
       bUseBrushTransparency=false
       roundingType=HalfHeightRadius
       width=0
       color {colorUseRule=UseColor_Specified, specifiedColor=<rgba 0 0 0 0>}
       cornerRadii=<0 0 0 0>
      }
      tintColor {colorUseRule=UseColor_Specified, specifiedColor=<rgba 0.0513 0.0953 0.159 1>}
      uVRegion {bIsValid=false, max=<0 0>, min=<0 0>}
     }
     inactiveHighlightedBrush {
      bIsDynamicallyLoaded=false
      drawAs=Image
      imageType=NoImage
      mirroring=NoMirror
      resourceName=None
      resourceObject=null
      tiling=NoTile
      imageSize=<0 0>
      margin=<ltrb 0 0 0 0>
      outlineSettings {
       bUseBrushTransparency=false
       roundingType=HalfHeightRadius
       width=0
       color {colorUseRule=UseColor_Specified, specifiedColor=<rgba 0 0 0 0>}
       cornerRadii=<0 0 0 0>
      }
      tintColor {colorUseRule=UseColor_Specified, specifiedColor=<rgba 0.0252 0.0319 0.0423 1>}
      uVRegion {bIsValid=false, max=<0 0>, min=<0 0>}
     }
     inactiveHoveredBrush {
      bIsDynamicallyLoaded=false
      drawAs=Image
      imageType=NoImage
      mirroring=NoMirror
      resourceName=None
      resourceObject=null
      tiling=NoTile
      imageSize=<0 0>
      margin=<ltrb 0 0 0 0>
      outlineSettings {
       bUseBrushTransparency=false
       roundingType=HalfHeightRadius
       width=0
       color {colorUseRule=UseColor_Specified, specifiedColor=<rgba 0 0 0 0>}
       cornerRadii=<0 0 0 0>
      }
      tintColor {colorUseRule=UseColor_Specified, specifiedColor=<rgba 0.0513 0.0953 0.159 1>}
      uVRegion {bIsValid=false, max=<0 0>, min=<0 0>}
     }
     oddRowBackgroundBrush {
      bIsDynamicallyLoaded=false
      drawAs=Image
      imageType=NoImage
      mirroring=NoMirror
      resourceName=None
      resourceObject=null
      tiling=NoTile
      imageSize=<0 0>
      margin=<ltrb 0 0 0 0>
      outlineSettings {
       bUseBrushTransparency=false
       roundingType=HalfHeightRadius
       width=0
       color {colorUseRule=UseColor_Specified, specifiedColor=<rgba 0 0 0 0>}
       cornerRadii=<0 0 0 0>
      }
      tintColor {colorUseRule=UseColor_Specified, specifiedColor=<rgba 0.0103 0.0103 0.0103 1>}
      uVRegion {bIsValid=false, max=<0 0>, min=<0 0>}
     }
     oddRowBackgroundHoveredBrush {
      bIsDynamicallyLoaded=false
      drawAs=Image
      imageType=NoImage
      mirroring=NoMirror
      resourceName=None
      resourceObject=null
      tiling=NoTile
      imageSize=<0 0>
      margin=<ltrb 0 0 0 0>
      outlineSettings {
       bUseBrushTransparency=false
       roundingType=HalfHeightRadius
       width=0
       color {colorUseRule=UseColor_Specified, specifiedColor=<rgba 0 0 0 0>}
       cornerRadii=<0 0 0 0>
      }
      tintColor {colorUseRule=UseColor_Specified, specifiedColor=<rgba 0.0176 0.0176 0.0176 1>}
      uVRegion {bIsValid=false, max=<0 0>, min=<0 0>}
     }
     parentRowBackgroundBrush {
      bIsDynamicallyLoaded=false
      drawAs=Image
      imageType=NoImage
      mirroring=NoMirror
      resourceName=None
      resourceObject=null
      tiling=NoTile
      imageSize=<32 32>
      margin=<ltrb 0 0 0 0>
      outlineSettings {
       bUseBrushTransparency=false
       roundingType=HalfHeightRadius
       width=0
       color {colorUseRule=UseColor_Specified, specifiedColor=<rgba 0 0 0 0>}
       cornerRadii=<0 0 0 0>
      }
      tintColor {colorUseRule=UseColor_Specified, specifiedColor=<rgba 1 1 1 1>}
      uVRegion {bIsValid=false, max=<0 0>, min=<0 0>}
     }
     parentRowBackgroundHoveredBrush {
      bIsDynamicallyLoaded=false
      drawAs=Image
      imageType=NoImage
      mirroring=NoMirror
      resourceName=None
      resourceObject=null
      tiling=NoTile
      imageSize=<32 32>
      margin=<ltrb 0 0 0 0>
      outlineSettings {
       bUseBrushTransparency=false
       roundingType=HalfHeightRadius
       width=0
       color {colorUseRule=UseColor_Specified, specifiedColor=<rgba 0 0 0 0>}
       cornerRadii=<0 0 0 0>
      }
      tintColor {colorUseRule=UseColor_Specified, specifiedColor=<rgba 1 1 1 1>}
      uVRegion {bIsValid=false, max=<0 0>, min=<0 0>}
     }
     selectedTextColor {colorUseRule=UseColor_Specified, specifiedColor=<rgba 1 1 1 1>}
     selectorFocusedBrush {
      bIsDynamicallyLoaded=false
      drawAs=Border
      imageType=FullColor
      mirroring=NoMirror
      resourceName=../../../Engine/Content/Slate/Common/Selector.png
      resourceObject=null
      tiling=Both
      imageSize=<0 0>
      margin=<ltrb 0.25 0.25 0.25 0.25>
      outlineSettings {
       bUseBrushTransparency=false
       roundingType=HalfHeightRadius
       width=0
       color {colorUseRule=UseColor_Specified, specifiedColor=<rgba 0 0 0 0>}
       cornerRadii=<0 0 0 0>
      }
      tintColor {colorUseRule=UseColor_Specified, specifiedColor=<rgba 0.701 0.225 0.003 1>}
      uVRegion {bIsValid=false, max=<0 0>, min=<0 0>}
     }
     textColor {colorUseRule=UseColor_Specified, specifiedColor=<rgba 0.5271 0.5271 0.5271 1>}
    }
    widgetStyle {
     comboButtonStyle {
      downArrowAlign=VAlign_Center
      buttonStyle {
       clickedSlateSound {resourceObject=null}
       disabled {
        bIsDynamicallyLoaded=false
        drawAs=RoundedBox
        imageType=NoImage
        mirroring=NoMirror
        resourceName=None
        resourceObject=null
        tiling=NoTile
        imageSize=<0 0>
        margin=<ltrb 0 0 0 0>
        outlineSettings {
         bUseBrushTransparency=false
         roundingType=FixedRadius
         width=1
         color {colorUseRule=UseColor_ColorTable, specifiedColor=<rgba 0 0 0 0>}
         cornerRadii=<4 4 4 4>
        }
        tintColor {colorUseRule=UseColor_Specified, specifiedColor=<rgba 0.0395 0.0395 0.0395 1>}
        uVRegion {bIsValid=false, max=<0 0>, min=<0 0>}
       }
       disabledForeground {colorUseRule=UseColor_ColorTable, specifiedColor=<rgba 0 0 0 0>}
       hovered {
        bIsDynamicallyLoaded=false
        drawAs=RoundedBox
        imageType=NoImage
        mirroring=NoMirror
        resourceName=None
        resourceObject=null
        tiling=NoTile
        imageSize=<0 0>
        margin=<ltrb 0 0 0 0>
        outlineSettings {
         bUseBrushTransparency=false
         roundingType=FixedRadius
         width=1
         color {colorUseRule=UseColor_ColorTable, specifiedColor=<rgba 0 0 0 0>}
         cornerRadii=<4 4 4 4>
        }
        tintColor {colorUseRule=UseColor_Specified, specifiedColor=<rgba 0.0048 0.0048 0.0048 1>}
        uVRegion {bIsValid=false, max=<0 0>, min=<0 0>}
       }
       hoveredForeground {colorUseRule=UseColor_ColorTable, specifiedColor=<rgba 0 0 0 0>}
       hoveredSlateSound {resourceObject=null}
       normal {
        bIsDynamicallyLoaded=false
        drawAs=RoundedBox
        imageType=NoImage
        mirroring=NoMirror
        resourceName=None
        resourceObject=null
        tiling=NoTile
        imageSize=<0 0>
        margin=<ltrb 0 0 0 0>
        outlineSettings {
         bUseBrushTransparency=false
         roundingType=FixedRadius
         width=1
         color {colorUseRule=UseColor_ColorTable, specifiedColor=<rgba 0 0 0 0>}
         cornerRadii=<4 4 4 4>
        }
        tintColor {colorUseRule=UseColor_Specified, specifiedColor=<rgba 0.0048 0.0048 0.0048 1>}
        uVRegion {bIsValid=false, max=<0 0>, min=<0 0>}
       }
       normalForeground {colorUseRule=UseColor_ColorTable, specifiedColor=<rgba 0 0 0 0>}
       normalPadding=<ltrb 8 1 8 1>
       pressed {
        bIsDynamicallyLoaded=false
        drawAs=RoundedBox
        imageType=NoImage
        mirroring=NoMirror
        resourceName=None
        resourceObject=null
        tiling=NoTile
        imageSize=<0 0>
        margin=<ltrb 0 0 0 0>
        outlineSettings {
         bUseBrushTransparency=false
         roundingType=FixedRadius
         width=1
         color {colorUseRule=UseColor_ColorTable, specifiedColor=<rgba 0 0 0 0>}
         cornerRadii=<4 4 4 4>
        }
        tintColor {colorUseRule=UseColor_Specified, specifiedColor=<rgba 0.0048 0.0048 0.0048 1>}
        uVRegion {bIsValid=false, max=<0 0>, min=<0 0>}
       }
       pressedForeground {colorUseRule=UseColor_ColorTable, specifiedColor=<rgba 0 0 0 0>}
       pressedPadding=<ltrb 8 2 8 0>
       pressedSlateSound {resourceObject=null}
      }
      contentPadding=<ltrb 0 0 0 0>
      downArrowImage {
       bIsDynamicallyLoaded=false
       drawAs=Image
       imageType=Vector
       mirroring=NoMirror
       resourceName=../../../Engine/Content/Slate/Starship/CoreWidgets/ComboBox/wide-chevron-down.svg
       resourceObject=null
       tiling=NoTile
       imageSize=<16 16>
       margin=<ltrb 0 0 0 0>
       outlineSettings {
        bUseBrushTransparency=false
        roundingType=HalfHeightRadius
        width=0
        color {colorUseRule=UseColor_Specified, specifiedColor=<rgba 0 0 0 0>}
        cornerRadii=<0 0 0 0>
       }
       tintColor {colorUseRule=UseColor_Specified, specifiedColor=<rgba 1 1 1 1>}
       uVRegion {bIsValid=false, max=<0 0>, min=<0 0>}
      }
      downArrowPadding=<ltrb 2 2 2 2>
      menuBorderBrush {
       bIsDynamicallyLoaded=false
       drawAs=RoundedBox
       imageType=NoImage
       mirroring=NoMirror
       resourceName=None
       resourceObject=null
       tiling=NoTile
       imageSize=<0 0>
       margin=<ltrb 0 0 0 0>
       outlineSettings {
        bUseBrushTransparency=false
        roundingType=FixedRadius
        width=1
        color {colorUseRule=UseColor_ColorTable, specifiedColor=<rgba 0 0 0 0>}
        cornerRadii=<0 0 0 0>
       }
       tintColor {colorUseRule=UseColor_Specified, specifiedColor=<rgba 0.0395 0.0395 0.0395 1>}
       uVRegion {bIsValid=false, max=<0 0>, min=<0 0>}
      }
      menuBorderPadding=<ltrb 0 0 0 0>
      shadowColorAndOpacity=<rgba 0 0 0 1>
      shadowOffset=<0 0>
     }
     contentPadding=<ltrb 0 0 0 0>
     menuRowPadding=<ltrb 8 3 8 3>
     pressedSlateSound {resourceObject=null}
     selectionChangeSlateSound {resourceObject=null}
    }
   }
   slot {class=SizeBoxSlot}
  }
  OtherOptionsBox {
   widgetClassPath=&HorizontalBox
   bIsVariable=true
   parent=VerticalBox_1
   widget=OtherOptionsBox
   slot {class=VerticalBoxSlot, properties {horizontalAlignment=HAlign_Center, verticalAlignment=VAlign_Center}}
  }
  SelectAllCheckBox {
   widgetClassPath=&EditorUtilityCheckBox
   bIsVariable=true
   parent=OtherOptionsBox
   widget=SelectAllCheckBox
   slot {class=HorizontalBoxSlot}
  }
  TextBlock_57 {
   widgetClassPath=&TextBlock
   parent=SelectAllCheckBox
   widget=TextBlock_57
   properties {
    text=Select All
    font {
     bForceMonospaced=false
     bMaterialIsStencil=false
     fontMaterial=null
     fontObject=&Roboto
     letterSpacing=0
     monospacedWidth=1
     size=13.5
     skewAmount=0
     typefaceFontName=Bold
     outlineSettings {
      bApplyOutlineToDropShadows=false
      bMiteredCorners=false
      bSeparateFillAlpha=false
      outlineMaterial=null
      outlineSize=0
      outlineColor=<rgba 0 0 0 1>
     }
    }
   }
   slot {class=PanelSlot}
  }
  Spacer_2 {widgetClassPath=&Spacer, parent=VerticalBox_1, widget=Spacer_2, slot {class=VerticalBoxSlot, properties {padding=<ltrb 0 5 0 5>}}}
  OtherOptionsBox_1 {
   widgetClassPath=&HorizontalBox
   bIsVariable=true
   parent=VerticalBox_1
   widget=OtherOptionsBox_1
   slot {class=VerticalBoxSlot, properties {horizontalAlignment=HAlign_Center, verticalAlignment=VAlign_Center}}
  }
  DataAssetCheckBox {
   widgetClassPath=&EditorUtilityCheckBox
   bIsVariable=true
   parent=OtherOptionsBox_1
   widget=DataAssetCheckBox
   slot {class=HorizontalBoxSlot}
  }
  TextBlock {
   widgetClassPath=&TextBlock
   parent=DataAssetCheckBox
   widget=TextBlock
   properties {
    text=Data Asset
    font {
     bForceMonospaced=false
     bMaterialIsStencil=false
     fontMaterial=null
     fontObject=&Roboto
     letterSpacing=0
     monospacedWidth=1
     size=13.5
     skewAmount=0
     typefaceFontName=Bold
     outlineSettings {
      bApplyOutlineToDropShadows=false
      bMiteredCorners=false
      bSeparateFillAlpha=false
      outlineMaterial=null
      outlineSize=0
      outlineColor=<rgba 0 0 0 1>
     }
    }
   }
   slot {class=PanelSlot}
  }
  Spacer {
   widgetClassPath=&Spacer
   parent=OtherOptionsBox_1
   widget=Spacer
   slot {class=HorizontalBoxSlot, properties {padding=<ltrb 5 0 5 0>}}
  }
  AnimBPCheckBox {
   widgetClassPath=&EditorUtilityCheckBox
   bIsVariable=true
   parent=OtherOptionsBox_1
   widget=AnimBPCheckBox
   slot {class=HorizontalBoxSlot}
  }
  TextBlock_1 {
   widgetClassPath=&TextBlock
   parent=AnimBPCheckBox
   widget=TextBlock_1
   properties {
    text=Anim BP
    font {
     bForceMonospaced=false
     bMaterialIsStencil=false
     fontMaterial=null
     fontObject=&Roboto
     letterSpacing=0
     monospacedWidth=1
     size=13.5
     skewAmount=0
     typefaceFontName=Bold
     outlineSettings {
      bApplyOutlineToDropShadows=false
      bMiteredCorners=false
      bSeparateFillAlpha=false
      outlineMaterial=null
      outlineSize=0
      outlineColor=<rgba 0 0 0 1>
     }
    }
   }
   slot {class=PanelSlot}
  }
  Spacer_1 {
   widgetClassPath=&Spacer
   parent=OtherOptionsBox_1
   widget=Spacer_1
   slot {class=HorizontalBoxSlot, properties {padding=<ltrb 5 0 5 0>}}
  }
  MontageCheckBox {
   widgetClassPath=&EditorUtilityCheckBox
   bIsVariable=true
   parent=OtherOptionsBox_1
   widget=MontageCheckBox
   slot {class=HorizontalBoxSlot}
  }
  Montages {
   widgetClassPath=&TextBlock
   parent=MontageCheckBox
   widget=Montages
   properties {
    text=Montages
    font {
     bForceMonospaced=false
     bMaterialIsStencil=false
     fontMaterial=null
     fontObject=&Roboto
     letterSpacing=0
     monospacedWidth=1
     size=13.5
     skewAmount=0
     typefaceFontName=Bold
     outlineSettings {
      bApplyOutlineToDropShadows=false
      bMiteredCorners=false
      bSeparateFillAlpha=false
      outlineMaterial=null
      outlineSize=0
      outlineColor=<rgba 0 0 0 1>
     }
    }
   }
   slot {class=PanelSlot}
  }
  Spacer_3 {
   widgetClassPath=&Spacer
   parent=OtherOptionsBox_1
   widget=Spacer_3
   slot {class=HorizontalBoxSlot, properties {padding=<ltrb 5 0 5 0>}}
  }
  CASCheckBox {
   widgetClassPath=&EditorUtilityCheckBox
   bIsVariable=true
   parent=OtherOptionsBox_1
   widget=CASCheckBox
   slot {class=HorizontalBoxSlot}
  }
  text {
   widgetClassPath=&TextBlock
   parent=CASCheckBox
   widget=text
   properties {
    text=CAS
    font {
     bForceMonospaced=false
     bMaterialIsStencil=false
     fontMaterial=null
     fontObject=&Roboto
     letterSpacing=0
     monospacedWidth=1
     size=13.5
     skewAmount=0
     typefaceFontName=Bold
     outlineSettings {
      bApplyOutlineToDropShadows=false
      bMiteredCorners=false
      bSeparateFillAlpha=false
      outlineMaterial=null
      outlineSize=0
      outlineColor=<rgba 0 0 0 1>
     }
    }
   }
   slot {class=PanelSlot}
  }
  HorizontalBox_116 {
   widgetClassPath=&HorizontalBox
   parent=CanvasPanel
   widget=HorizontalBox_116
   slot {
    class=CanvasPanelSlot
    properties {bAutoSize=true, layoutData {alignment=<0 0>, anchors {maximum=<1 1>, minimum=<0 0>}, offsets=<ltrb 0 86.9203 -0.025 48.5786>}}
   }
  }
  VerticalScrollBox {
   widgetClassPath=/Script/FortniteUI.ThresholdScrollBox
   parent=HorizontalBox_116
   widget=VerticalScrollBox
   properties {
    consumeMouseWheel=Always
    widgetBarStyle {
     thickness=8
     draggedThumbImage {
      bIsDynamicallyLoaded=false
      drawAs=RoundedBox
      imageType=NoImage
      mirroring=NoMirror
      resourceName=None
      resourceObject=null
      tiling=NoTile
      imageSize=<0 0>
      margin=<ltrb 0 0 0 0>
      outlineSettings {
       bUseBrushTransparency=false
       roundingType=FixedRadius
       width=0
       color {colorUseRule=UseColor_Specified, specifiedColor=<rgba 0 0 0 0>}
       cornerRadii=<4 4 4 4>
      }
      tintColor {colorUseRule=UseColor_Specified, specifiedColor=<rgba 0.2159 0.2159 0.2159 1>}
      uVRegion {bIsValid=false, max=<0 0>, min=<0 0>}
     }
     horizontalBackgroundImage {
      bIsDynamicallyLoaded=false
      drawAs=NoDrawType
      imageType=NoImage
      mirroring=NoMirror
      resourceName=None
      resourceObject=null
      tiling=NoTile
      imageSize=<0 0>
      margin=<ltrb 0 0 0 0>
      outlineSettings {
       bUseBrushTransparency=false
       roundingType=HalfHeightRadius
       width=0
       color {colorUseRule=UseColor_Specified, specifiedColor=<rgba 0 0 0 0>}
       cornerRadii=<0 0 0 0>
      }
      tintColor {colorUseRule=UseColor_Specified, specifiedColor=<rgba 1 1 1 1>}
      uVRegion {bIsValid=false, max=<0 0>, min=<0 0>}
     }
     horizontalBottomSlotImage {
      bIsDynamicallyLoaded=false
      drawAs=NoDrawType
      imageType=NoImage
      mirroring=NoMirror
      resourceName=None
      resourceObject=null
      tiling=NoTile
      imageSize=<0 0>
      margin=<ltrb 0 0 0 0>
      outlineSettings {
       bUseBrushTransparency=false
       roundingType=HalfHeightRadius
       width=0
       color {colorUseRule=UseColor_Specified, specifiedColor=<rgba 0 0 0 0>}
       cornerRadii=<0 0 0 0>
      }
      tintColor {colorUseRule=UseColor_Specified, specifiedColor=<rgba 1 1 1 1>}
      uVRegion {bIsValid=false, max=<0 0>, min=<0 0>}
     }
     horizontalTopSlotImage {
      bIsDynamicallyLoaded=false
      drawAs=NoDrawType
      imageType=NoImage
      mirroring=NoMirror
      resourceName=None
      resourceObject=null
      tiling=NoTile
      imageSize=<0 0>
      margin=<ltrb 0 0 0 0>
      outlineSettings {
       bUseBrushTransparency=false
       roundingType=HalfHeightRadius
       width=0
       color {colorUseRule=UseColor_Specified, specifiedColor=<rgba 0 0 0 0>}
       cornerRadii=<0 0 0 0>
      }
      tintColor {colorUseRule=UseColor_Specified, specifiedColor=<rgba 1 1 1 1>}
      uVRegion {bIsValid=false, max=<0 0>, min=<0 0>}
     }
     hoveredThumbImage {
      bIsDynamicallyLoaded=false
      drawAs=RoundedBox
      imageType=NoImage
      mirroring=NoMirror
      resourceName=None
      resourceObject=null
      tiling=NoTile
      imageSize=<0 0>
      margin=<ltrb 0 0 0 0>
      outlineSettings {
       bUseBrushTransparency=false
       roundingType=FixedRadius
       width=0
       color {colorUseRule=UseColor_Specified, specifiedColor=<rgba 0 0 0 0>}
       cornerRadii=<4 4 4 4>
      }
      tintColor {colorUseRule=UseColor_Specified, specifiedColor=<rgba 0.2159 0.2159 0.2159 1>}
      uVRegion {bIsValid=false, max=<0 0>, min=<0 0>}
     }
     normalThumbImage {
      bIsDynamicallyLoaded=false
      drawAs=RoundedBox
      imageType=NoImage
      mirroring=NoMirror
      resourceName=None
      resourceObject=null
      tiling=NoTile
      imageSize=<0 0>
      margin=<ltrb 0 0 0 0>
      outlineSettings {
       bUseBrushTransparency=false
       roundingType=FixedRadius
       width=0
       color {colorUseRule=UseColor_Specified, specifiedColor=<rgba 0 0 0 0>}
       cornerRadii=<4 4 4 4>
      }
      tintColor {colorUseRule=UseColor_Specified, specifiedColor=<rgba 0.0953 0.0953 0.0953 1>}
      uVRegion {bIsValid=false, max=<0 0>, min=<0 0>}
     }
     verticalBackgroundImage {
      bIsDynamicallyLoaded=false
      drawAs=NoDrawType
      imageType=NoImage
      mirroring=NoMirror
      resourceName=None
      resourceObject=null
      tiling=NoTile
      imageSize=<0 0>
      margin=<ltrb 0 0 0 0>
      outlineSettings {
       bUseBrushTransparency=false
       roundingType=HalfHeightRadius
       width=0
       color {colorUseRule=UseColor_Specified, specifiedColor=<rgba 0 0 0 0>}
       cornerRadii=<0 0 0 0>
      }
      tintColor {colorUseRule=UseColor_Specified, specifiedColor=<rgba 1 1 1 1>}
      uVRegion {bIsValid=false, max=<0 0>, min=<0 0>}
     }
     verticalBottomSlotImage {
      bIsDynamicallyLoaded=false
      drawAs=NoDrawType
      imageType=NoImage
      mirroring=NoMirror
      resourceName=None
      resourceObject=null
      tiling=NoTile
      imageSize=<0 0>
      margin=<ltrb 0 0 0 0>
      outlineSettings {
       bUseBrushTransparency=false
       roundingType=HalfHeightRadius
       width=0
       color {colorUseRule=UseColor_Specified, specifiedColor=<rgba 0 0 0 0>}
       cornerRadii=<0 0 0 0>
      }
      tintColor {colorUseRule=UseColor_Specified, specifiedColor=<rgba 1 1 1 1>}
      uVRegion {bIsValid=false, max=<0 0>, min=<0 0>}
     }
     verticalTopSlotImage {
      bIsDynamicallyLoaded=false
      drawAs=NoDrawType
      imageType=NoImage
      mirroring=NoMirror
      resourceName=None
      resourceObject=null
      tiling=NoTile
      imageSize=<0 0>
      margin=<ltrb 0 0 0 0>
      outlineSettings {
       bUseBrushTransparency=false
       roundingType=HalfHeightRadius
       width=0
       color {colorUseRule=UseColor_Specified, specifiedColor=<rgba 0 0 0 0>}
       cornerRadii=<0 0 0 0>
      }
      tintColor {colorUseRule=UseColor_Specified, specifiedColor=<rgba 1 1 1 1>}
      uVRegion {bIsValid=false, max=<0 0>, min=<0 0>}
     }
    }
    widgetStyle {
     barThickness=8
     bottomShadowBrush {
      bIsDynamicallyLoaded=false
      drawAs=Box
      imageType=FullColor
      mirroring=NoMirror
      resourceName=../../../Engine/Content/Slate/Common/ScrollBoxShadowBottom.png
      resourceObject=null
      tiling=NoTile
      imageSize=<16 8>
      margin=<ltrb 0.5 0 0.5 1>
      outlineSettings {
       bUseBrushTransparency=false
       roundingType=HalfHeightRadius
       width=0
       color {colorUseRule=UseColor_Specified, specifiedColor=<rgba 0 0 0 0>}
       cornerRadii=<0 0 0 0>
      }
      tintColor {colorUseRule=UseColor_Specified, specifiedColor=<rgba 1 1 1 1>}
      uVRegion {bIsValid=false, max=<0 0>, min=<0 0>}
     }
     horizontalScrolledContentPadding=<ltrb 0 0 1 0>
     leftShadowBrush {
      bIsDynamicallyLoaded=false
      drawAs=Box
      imageType=FullColor
      mirroring=NoMirror
      resourceName=../../../Engine/Content/Slate/Common/ScrollBoxShadowLeft.png
      resourceObject=null
      tiling=NoTile
      imageSize=<8 16>
      margin=<ltrb 1 0.5 0 0.5>
      outlineSettings {
       bUseBrushTransparency=false
       roundingType=HalfHeightRadius
       width=0
       color {colorUseRule=UseColor_Specified, specifiedColor=<rgba 0 0 0 0>}
       cornerRadii=<0 0 0 0>
      }
      tintColor {colorUseRule=UseColor_Specified, specifiedColor=<rgba 1 1 1 1>}
      uVRegion {bIsValid=false, max=<0 0>, min=<0 0>}
     }
     rightShadowBrush {
      bIsDynamicallyLoaded=false
      drawAs=Box
      imageType=FullColor
      mirroring=NoMirror
      resourceName=../../../Engine/Content/Slate/Common/ScrollBoxShadowRight.png
      resourceObject=null
      tiling=NoTile
      imageSize=<8 16>
      margin=<ltrb 0 0.5 1 0.5>
      outlineSettings {
       bUseBrushTransparency=false
       roundingType=HalfHeightRadius
       width=0
       color {colorUseRule=UseColor_Specified, specifiedColor=<rgba 0 0 0 0>}
       cornerRadii=<0 0 0 0>
      }
      tintColor {colorUseRule=UseColor_Specified, specifiedColor=<rgba 1 1 1 1>}
      uVRegion {bIsValid=false, max=<0 0>, min=<0 0>}
     }
     topShadowBrush {
      bIsDynamicallyLoaded=false
      drawAs=Box
      imageType=FullColor
      mirroring=NoMirror
      resourceName=../../../Engine/Content/Slate/Common/ScrollBoxShadowTop.png
      resourceObject=null
      tiling=NoTile
      imageSize=<16 8>
      margin=<ltrb 0.5 1 0.5 0>
      outlineSettings {
       bUseBrushTransparency=false
       roundingType=HalfHeightRadius
       width=0
       color {colorUseRule=UseColor_Specified, specifiedColor=<rgba 0 0 0 0>}
       cornerRadii=<0 0 0 0>
      }
      tintColor {colorUseRule=UseColor_Specified, specifiedColor=<rgba 1 1 1 1>}
      uVRegion {bIsValid=false, max=<0 0>, min=<0 0>}
     }
     verticalScrolledContentPadding=<ltrb 0 0 0 1>
    }
   }
   slot {class=HorizontalBoxSlot, properties {size {sizeRule=Fill, value=1}}}
  }
  HorizontalScrollbox {
   widgetClassPath=&ThresholdScrollBox
   parent=VerticalScrollBox
   widget=HorizontalScrollbox
   properties {
    consumeMouseWheel=Never
    orientation=Orient_Horizontal
    widgetBarStyle {
     thickness=8
     draggedThumbImage {
      bIsDynamicallyLoaded=false
      drawAs=RoundedBox
      imageType=NoImage
      mirroring=NoMirror
      resourceName=None
      resourceObject=null
      tiling=NoTile
      imageSize=<0 0>
      margin=<ltrb 0 0 0 0>
      outlineSettings {
       bUseBrushTransparency=false
       roundingType=FixedRadius
       width=0
       color {colorUseRule=UseColor_Specified, specifiedColor=<rgba 0 0 0 0>}
       cornerRadii=<4 4 4 4>
      }
      tintColor {colorUseRule=UseColor_Specified, specifiedColor=<rgba 0.2159 0.2159 0.2159 1>}
      uVRegion {bIsValid=false, max=<0 0>, min=<0 0>}
     }
     horizontalBackgroundImage {
      bIsDynamicallyLoaded=false
      drawAs=NoDrawType
      imageType=NoImage
      mirroring=NoMirror
      resourceName=None
      resourceObject=null
      tiling=NoTile
      imageSize=<0 0>
      margin=<ltrb 0 0 0 0>
      outlineSettings {
       bUseBrushTransparency=false
       roundingType=HalfHeightRadius
       width=0
       color {colorUseRule=UseColor_Specified, specifiedColor=<rgba 0 0 0 0>}
       cornerRadii=<0 0 0 0>
      }
      tintColor {colorUseRule=UseColor_Specified, specifiedColor=<rgba 1 1 1 1>}
      uVRegion {bIsValid=false, max=<0 0>, min=<0 0>}
     }
     horizontalBottomSlotImage {
      bIsDynamicallyLoaded=false
      drawAs=NoDrawType
      imageType=NoImage
      mirroring=NoMirror
      resourceName=None
      resourceObject=null
      tiling=NoTile
      imageSize=<0 0>
      margin=<ltrb 0 0 0 0>
      outlineSettings {
       bUseBrushTransparency=false
       roundingType=HalfHeightRadius
       width=0
       color {colorUseRule=UseColor_Specified, specifiedColor=<rgba 0 0 0 0>}
       cornerRadii=<0 0 0 0>
      }
      tintColor {colorUseRule=UseColor_Specified, specifiedColor=<rgba 1 1 1 1>}
      uVRegion {bIsValid=false, max=<0 0>, min=<0 0>}
     }
     horizontalTopSlotImage {
      bIsDynamicallyLoaded=false
      drawAs=NoDrawType
      imageType=NoImage
      mirroring=NoMirror
      resourceName=None
      resourceObject=null
      tiling=NoTile
      imageSize=<0 0>
      margin=<ltrb 0 0 0 0>
      outlineSettings {
       bUseBrushTransparency=false
       roundingType=HalfHeightRadius
       width=0
       color {colorUseRule=UseColor_Specified, specifiedColor=<rgba 0 0 0 0>}
       cornerRadii=<0 0 0 0>
      }
      tintColor {colorUseRule=UseColor_Specified, specifiedColor=<rgba 1 1 1 1>}
      uVRegion {bIsValid=false, max=<0 0>, min=<0 0>}
     }
     hoveredThumbImage {
      bIsDynamicallyLoaded=false
      drawAs=RoundedBox
      imageType=NoImage
      mirroring=NoMirror
      resourceName=None
      resourceObject=null
      tiling=NoTile
      imageSize=<0 0>
      margin=<ltrb 0 0 0 0>
      outlineSettings {
       bUseBrushTransparency=false
       roundingType=FixedRadius
       width=0
       color {colorUseRule=UseColor_Specified, specifiedColor=<rgba 0 0 0 0>}
       cornerRadii=<4 4 4 4>
      }
      tintColor {colorUseRule=UseColor_Specified, specifiedColor=<rgba 0.2159 0.2159 0.2159 1>}
      uVRegion {bIsValid=false, max=<0 0>, min=<0 0>}
     }
     normalThumbImage {
      bIsDynamicallyLoaded=false
      drawAs=RoundedBox
      imageType=NoImage
      mirroring=NoMirror
      resourceName=None
      resourceObject=null
      tiling=NoTile
      imageSize=<0 0>
      margin=<ltrb 0 0 0 0>
      outlineSettings {
       bUseBrushTransparency=false
       roundingType=FixedRadius
       width=0
       color {colorUseRule=UseColor_Specified, specifiedColor=<rgba 0 0 0 0>}
       cornerRadii=<4 4 4 4>
      }
      tintColor {colorUseRule=UseColor_Specified, specifiedColor=<rgba 0.0953 0.0953 0.0953 1>}
      uVRegion {bIsValid=false, max=<0 0>, min=<0 0>}
     }
     verticalBackgroundImage {
      bIsDynamicallyLoaded=false
      drawAs=NoDrawType
      imageType=NoImage
      mirroring=NoMirror
      resourceName=None
      resourceObject=null
      tiling=NoTile
      imageSize=<0 0>
      margin=<ltrb 0 0 0 0>
      outlineSettings {
       bUseBrushTransparency=false
       roundingType=HalfHeightRadius
       width=0
       color {colorUseRule=UseColor_Specified, specifiedColor=<rgba 0 0 0 0>}
       cornerRadii=<0 0 0 0>
      }
      tintColor {colorUseRule=UseColor_Specified, specifiedColor=<rgba 1 1 1 1>}
      uVRegion {bIsValid=false, max=<0 0>, min=<0 0>}
     }
     verticalBottomSlotImage {
      bIsDynamicallyLoaded=false
      drawAs=NoDrawType
      imageType=NoImage
      mirroring=NoMirror
      resourceName=None
      resourceObject=null
      tiling=NoTile
      imageSize=<0 0>
      margin=<ltrb 0 0 0 0>
      outlineSettings {
       bUseBrushTransparency=false
       roundingType=HalfHeightRadius
       width=0
       color {colorUseRule=UseColor_Specified, specifiedColor=<rgba 0 0 0 0>}
       cornerRadii=<0 0 0 0>
      }
      tintColor {colorUseRule=UseColor_Specified, specifiedColor=<rgba 1 1 1 1>}
      uVRegion {bIsValid=false, max=<0 0>, min=<0 0>}
     }
     verticalTopSlotImage {
      bIsDynamicallyLoaded=false
      drawAs=NoDrawType
      imageType=NoImage
      mirroring=NoMirror
      resourceName=None
      resourceObject=null
      tiling=NoTile
      imageSize=<0 0>
      margin=<ltrb 0 0 0 0>
      outlineSettings {
       bUseBrushTransparency=false
       roundingType=HalfHeightRadius
       width=0
       color {colorUseRule=UseColor_Specified, specifiedColor=<rgba 0 0 0 0>}
       cornerRadii=<0 0 0 0>
      }
      tintColor {colorUseRule=UseColor_Specified, specifiedColor=<rgba 1 1 1 1>}
      uVRegion {bIsValid=false, max=<0 0>, min=<0 0>}
     }
    }
    widgetStyle {
     barThickness=8
     bottomShadowBrush {
      bIsDynamicallyLoaded=false
      drawAs=Box
      imageType=FullColor
      mirroring=NoMirror
      resourceName=../../../Engine/Content/Slate/Common/ScrollBoxShadowBottom.png
      resourceObject=null
      tiling=NoTile
      imageSize=<16 8>
      margin=<ltrb 0.5 0 0.5 1>
      outlineSettings {
       bUseBrushTransparency=false
       roundingType=HalfHeightRadius
       width=0
       color {colorUseRule=UseColor_Specified, specifiedColor=<rgba 0 0 0 0>}
       cornerRadii=<0 0 0 0>
      }
      tintColor {colorUseRule=UseColor_Specified, specifiedColor=<rgba 1 1 1 1>}
      uVRegion {bIsValid=false, max=<0 0>, min=<0 0>}
     }
     horizontalScrolledContentPadding=<ltrb 0 0 1 0>
     leftShadowBrush {
      bIsDynamicallyLoaded=false
      drawAs=Box
      imageType=FullColor
      mirroring=NoMirror
      resourceName=../../../Engine/Content/Slate/Common/ScrollBoxShadowLeft.png
      resourceObject=null
      tiling=NoTile
      imageSize=<8 16>
      margin=<ltrb 1 0.5 0 0.5>
      outlineSettings {
       bUseBrushTransparency=false
       roundingType=HalfHeightRadius
       width=0
       color {colorUseRule=UseColor_Specified, specifiedColor=<rgba 0 0 0 0>}
       cornerRadii=<0 0 0 0>
      }
      tintColor {colorUseRule=UseColor_Specified, specifiedColor=<rgba 1 1 1 1>}
      uVRegion {bIsValid=false, max=<0 0>, min=<0 0>}
     }
     rightShadowBrush {
      bIsDynamicallyLoaded=false
      drawAs=Box
      imageType=FullColor
      mirroring=NoMirror
      resourceName=../../../Engine/Content/Slate/Common/ScrollBoxShadowRight.png
      resourceObject=null
      tiling=NoTile
      imageSize=<8 16>
      margin=<ltrb 0 0.5 1 0.5>
      outlineSettings {
       bUseBrushTransparency=false
       roundingType=HalfHeightRadius
       width=0
       color {colorUseRule=UseColor_Specified, specifiedColor=<rgba 0 0 0 0>}
       cornerRadii=<0 0 0 0>
      }
      tintColor {colorUseRule=UseColor_Specified, specifiedColor=<rgba 1 1 1 1>}
      uVRegion {bIsValid=false, max=<0 0>, min=<0 0>}
     }
     topShadowBrush {
      bIsDynamicallyLoaded=false
      drawAs=Box
      imageType=FullColor
      mirroring=NoMirror
      resourceName=../../../Engine/Content/Slate/Common/ScrollBoxShadowTop.png
      resourceObject=null
      tiling=NoTile
      imageSize=<16 8>
      margin=<ltrb 0.5 1 0.5 0>
      outlineSettings {
       bUseBrushTransparency=false
       roundingType=HalfHeightRadius
       width=0
       color {colorUseRule=UseColor_Specified, specifiedColor=<rgba 0 0 0 0>}
       cornerRadii=<0 0 0 0>
      }
      tintColor {colorUseRule=UseColor_Specified, specifiedColor=<rgba 1 1 1 1>}
      uVRegion {bIsValid=false, max=<0 0>, min=<0 0>}
     }
     verticalScrolledContentPadding=<ltrb 0 0 0 1>
    }
   }
   slot {class=ScrollBoxSlot, properties {size {sizeRule=Fill, value=1}}}
  }
  SequenceColumnHolder {
   widgetClassPath=&StackBox
   bIsVariable=true
   parent=HorizontalScrollbox
   widget=SequenceColumnHolder
   properties {bOverride_Cursor=true, visibility=Visible}
   slot {class=ScrollBoxSlot, properties {padding=<ltrb 3 3 3 3>, size {sizeRule=Automatic, value=2}}}
  }
  BlendSpaceColumnHolder {
   widgetClassPath=&StackBox
   bIsVariable=true
   parent=HorizontalScrollbox
   widget=BlendSpaceColumnHolder
   properties {bOverride_Cursor=true, visibility=Visible}
   slot {class=ScrollBoxSlot, properties {padding=<ltrb 3 3 3 3>, size {sizeRule=Automatic, value=2}}}
  }
  ChooserTableColumnHolder {
   widgetClassPath=&StackBox
   bIsVariable=true
   parent=HorizontalScrollbox
   widget=ChooserTableColumnHolder
   properties {bOverride_Cursor=true, visibility=Visible}
   slot {class=ScrollBoxSlot, properties {padding=<ltrb 3 3 3 3>, size {sizeRule=Automatic, value=2}}}
  }
  HorizontalBox {
   widgetClassPath=&HorizontalBox
   parent=CanvasPanel
   widget=HorizontalBox
   slot {
    class=CanvasPanelSlot
    properties {layoutData {alignment=<0 0>, anchors {maximum=<1 1>, minimum=<0 1>}, offsets=<ltrb 0.7669 -48.8457 0.8584 49.1302>}}
   }
  }
  GenerateBtn {
   widgetClassPath=&EditorUtilityButton
   parent=HorizontalBox
   widget=GenerateBtn
   slot {class=HorizontalBoxSlot, properties {size {sizeRule=Fill, value=1}}}
  }
  TextBlock_53 {
   widgetClassPath=&TextBlock
   parent=GenerateBtn
   widget=TextBlock_53
   properties {justification=Center, text=Generate Assets}
   slot {class=ButtonSlot}
  }
  ErrorTxt {
   widgetClassPath=&TextBlock
   bIsVariable=true
   parent=CanvasPanel
   widget=ErrorTxt
   properties {
    text=error text
    colorAndOpacity {colorUseRule=UseColor_Specified, specifiedColor=<rgba 1 0 0 1>}
    font {
     bForceMonospaced=false
     bMaterialIsStencil=false
     fontMaterial=null
     fontObject=&Roboto
     letterSpacing=0
     monospacedWidth=1
     size=12
     skewAmount=0
     typefaceFontName=Bold
     outlineSettings {
      bApplyOutlineToDropShadows=false
      bMiteredCorners=false
      bSeparateFillAlpha=false
      outlineMaterial=null
      outlineSize=0
      outlineColor=<rgba 0 0 0 1>
     }
    }
   }
   slot {class=CanvasPanelSlot, properties {layoutData {alignment=<0 0>, anchors {maximum=<0 0>, minimum=<0 0>}, offsets=<ltrb 12 40 100 30>}}}
  }
 ]
}
